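# Test material in the foliage base material folder that sets up debug readouts of a mesh's local bounds.
# The Max of PreSkinnedLocalBounds feeds both DebugFloat3Values and DebugScalarValues, the latter positioned by screen viewport UVs.
# ObjectLocalBounds, a vector parameter, a constant and a component mask sit unconnected, and no material output hookup is shown.
# Max World Position Offset displacement is capped, though no offset logic is present.
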Material /WildEstate/Foliage/BaseMaterial/M_TestLocalBound
graph {
 properties {bAutomaticallySetUsageInEditor=false, maxWorldPositionOffsetDisplacement=500, usedShadingModels=Default Lit}
 nodes [
  ComponentMask() {b=true}
  Constant3Vector()
  n1 = PreSkinnedLocalBounds()
  MaterialFunctionCall!(Vector3=@n1.Max, MaximumNumberOfDigits, UVs, DebugTextLocation RG_UpperRight BA_LowerLeft, Component Spacing) {materialFunction=/Engine/Functions/Engine_MaterialFunctions02/Utility/DebugFloat3Values}
  MaterialFunctionCall(Number=@n1.Max, UVs=ScreenPosition().ViewportUV) {materialFunction=&DebugScalarValues}
  ObjectLocalBounds()
  VectorParameter() {defaultValue=<rgba 500 0 0 1>}
 ]
}
