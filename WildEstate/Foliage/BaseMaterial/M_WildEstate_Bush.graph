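# Subsurface foliage material for Wild Estate bushes on instanced and Nanite meshes, with billboard-like deformation and stylized shading.
# World position offset math squeezes vertices along each instance's camera direction, fading with pixel depth and zeroed in shadow passes via ShadowReplace.
# A camera-facing virtual plane samples a leaf flipbook, picking frames by instance position.
# Floored height and sun-direction gradients, the latter noise-perturbed, form hard shadow bands that tint the color.
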Material /WildEstate/Foliage/BaseMaterial/M_WildEstate_Bush
graph {
 properties {
  bAutomaticallySetUsageInEditor=false
  bTangentSpaceNormal=false
  bUsedWithInstancedStaticMeshes=true
  bUsedWithNanite=true
  maxWorldPositionOffsetDisplacement=100
  shadingModel=MSM_Subsurface
  usedShadingModels=Subsurface
 }
 nodes [
  Bounds()
  CameraPositionWS()
  n1 = TransformPosition(None=Constant3Vector()) {transformSourceType=TRANSFORMPOSSOURCE_Instance, transformType=TRANSFORMPOSSOURCE_TranslatedWorld}
  n2 = Normalize(VectorInput=@n1)
  n3 = DotProduct(A=Subtract(A=WorldPosition() {worldPositionShaderOffset=WPT_CameraRelative}.XYZ, B=@n1), B=@n2)
  n4 = Max(A=Subtract(A=Multiply(A=@n3) {constB=-1}, B=ScalarParameter() {defaultValue=66.7177, parameterName=BillboardSqueze_Clamp}), B=Constant())
  Ceil(None=@n4)
  ComponentMask(None=TransformPosition(None=WorldPosition() {worldPositionShaderOffset=WPT_ExcludeAllShaderOffsets}.XYZ) {transformSourceType=TRANSFORMPOSSOURCE_World}) {b=true}
  n5 = Constant3Vector() {constant=<rgba -1 0 0 0>}
  n6 = NamedRerouteUsage()
  n7 = MaterialFunctionCall(
   Plane Normal Axis=CameraVectorWS()
   Plane X Axis=Normalize(VectorInput=Transform(None=@n5) {transformSourceType=TRANSFORMSOURCE_Camera})
   Plane Center=@n6
   UV Scale=Multiply(
    A=MaterialFunctionCall() {materialFunction=&ObjectScale}.Scale Z
    B=Constant() {r=1}
   )
  ) {materialFunction=&VirtualPlaneCoordinates}
  ComponentMask(None=@n7.World Position) {g=true, r=true}
  Constant3Vector()
  Constant3Vector()
  Constant3Vector() {constant=<rgba 0 0 -50 0>}
  Constant()
  Constant() {r=1}
  Constant()
  n8 = Constant3Vector() {constant=<rgba 0 0 1 0>}
  n9 = VertexNormalWS()
  Divide(A=Add(A=DotProduct(A=@n8, B=@n9)))
  DotProduct()
  n10 = Multiply(A=ScalarParameter() {defaultValue=-0.99, parameterName=BillboardSqueze}, B=Multiply(A=@n2, B=@n3) {constB=-1})
  n11 = Multiply(A=@n2, B=Saturate(None=@n4))
  n12 = ScalarParameter() {defaultValue=-13.3498, parameterName=BillboardSqueze_PushInner}
  n13 = StaticSwitchParameter(True=@n10, False=Add(A=@n10, B=Multiply(A=@n11, B=@n12))) {defaultValue=true, parameterName=UseVirtual Plane Card}
  n14 = Multiply(A=@n2, B=ScalarParameter() {defaultValue=-57.413, parameterName=BillboardOffsetCamDir})
  n15 = Divide(
   A=Subtract(A=PixelDepth().R, B=ScalarParameter() {defaultValue=4000, parameterName=WPO Disable Fade Offset})
   B=ScalarParameter() {defaultValue=1000, parameterName=WPO Disable Fade Blend}
  )
  LinearInterpolate(A=Constant3Vector(), B=ShadowReplace(Default=Add(A=@n13, B=@n14), Shadow=Constant3Vector()), Alpha=OneMinus(None=Saturate(None=@n15)))
  LocalPosition()
  n16 = Add(A=@n13, B=WorldPosition().XYZ)
  n17 = CameraPositionWS()
  Multiply(A=Normalize(VectorInput=Subtract(A=@n16, B=@n17)), B=ScalarParameter() {parameterName=ShadowOffset}) {constB=10}
  NamedRerouteDeclaration(None=TransformPosition(None=VectorParameter() {parameterName=ObjectCenter, defaultValue=<rgba 0 0 68.5129 1>}.RGB) {transformSourceType=TRANSFORMPOSSOURCE_Instance, transformType=TRANSFORMPOSSOURCE_World}) {name=ObjectPosition_Instance, nodeColor=<rgba 1 0 0.1172 1>}
  n18 = CollectionParameter() {
   collection=&FortniteMaterialParameters
   parameterName=SunlightDirectionalVector
  }
  n19 = ComponentMask(None=@n18) {b=true, g=true, r=true}
  Normalize(VectorInput=Add(A=Transform(None=@n19) {transformSourceType=TRANSFORMSOURCE_World}, B=CameraVectorWS()) {desc=Compromise for Direction light and all othe rlights})
  Normalize(VectorInput=Subtract(A=CameraPositionWS(), B=@n6))
  ObjectBounds()
  ObjectPositionWS()
  ObjectPositionWS()
  ScalarParameter() {parameterName=Opacity}
  n20 = Constant3Vector()
  n21 = MaterialFunctionCall(Float3=Divide(A=TransformPosition(None=@n20) {transformSourceType=TRANSFORMPOSSOURCE_Instance, transformType=TRANSFORMPOSSOURCE_World}) {constB=10}) {materialFunction=&BreakOutFloat3Components}
  n22 = Add(A=@n21.R, B=@n21.R)
  n23 = Constant() {r=2}
  n24 = Saturate(None=MaterialFunctionCall(UVs=@n7.UVs, Texture Scale=ScalarParameter() {defaultValue=296.9218, parameterName=VirtualPlaneUVSclae}) {materialFunction=&ScaleUVsByCenter}.UVs)
  n25 = MaterialFunctionCall(
   "Animation  Phase (0-1)"=Frac(None=Add(A=@n22, B=@n21.B))
   Number of Rows=@n23
   Number of Columns=@n23
   Texture=TextureObject() {texture=/WildEstate/Environment/TestContent/WildEstate_Bush_Leaf}
   UVs=@n24
  ) {materialFunction=&FlipBook}
  n26 = ScalarParameter() {defaultValue=0.5, parameterName=BottomShadowMin}
  n27 = LocalPosition()
  n28 = ScalarParameter() {defaultValue=-64.9245, parameterName=BottomShadowGradientOffset}
  n29 = TextureSample(UVs=Divide(A=@n7.UVs, B=ScalarParameter() {defaultValue=230.3388, parameterName=BushNoise_Scale})) {samplerType=SAMPLERTYPE_Masks, texture=/WildEstate/Environment/TestContent/T_WildEstateBush_Noise_M}
  n30 = Add(
   A=Divide(A=Add(A=@n27.Z, B=@n28), B=ScalarParameter() {defaultValue=147.2446, parameterName=BottomShadowGradientDist})
   B=Multiply(A=@n29.R, B=ScalarParameter() {defaultValue=0.2703, parameterName=NoiseAdd_LightDirectionGradient})
  )
  n31 = Saturate(None=Add(A=Saturate(None=@n30), B=ScalarParameter() {defaultValue=0.8469, parameterName=BottomShadowThreshold}))
  n32 = Subtract(A=WorldPosition().XYZ, B=NamedRerouteUsage())
  n33 = ScalarParameter() {defaultValue=0.2703, parameterName=NoiseAdd_LightDirectionGradient}
  n34 = Normalize(VectorInput=Add(A=Normalize(VectorInput=@n32), B=Multiply(A=@n29.R, B=@n33)))
  n35 = Saturate(None=Add(A=DotProduct(A=@n19, B=@n34), B=ScalarParameter() {defaultValue=1.3682, parameterName=Threshold}))
  n36 = Saturate(None=Multiply(A=Floor(None=@n31), B=Floor(None=@n35)))
  n37 = Multiply(A=Saturate(None=Add(A=@n26, B=@n36)), B=VectorParameter() {parameterName=Color, defaultValue=<rgba 0.0061 0.1354 0.0559 1>}.RGB)
  StaticSwitchParameter(True=Multiply(A=ComponentMask(None=@n25.Result) {r=true}, B=@n37), False=@n37) {defaultValue=true, parameterName=UseVirtual Plane Card}
  TextureSample(UVs=@n24) {texture=&WildEstate_Bush_Leaf}
 ]
}
