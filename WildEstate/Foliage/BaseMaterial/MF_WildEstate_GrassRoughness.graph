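# Material function in the Wild Estate foliage base materials that supplies a roughness value for grass.
# It outputs a fixed, very low constant.
# A pixel-depth gradient that would raise the value with distance from the camera is built in the graph but left unconnected, so it has no effect on the result.
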MaterialFunction /WildEstate/Foliage/BaseMaterial/MF_WildEstate_GrassRoughness
graph {nodes [
  Add(A=Divide(A=PixelDepth().R) {constB=70000}) {constB=0.01}
  FunctionOutput(None=Constant() {r=0.02})
 ]}
