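# Material function that animates medium foliage plants through World Position Offset, driven by wind, physics velocity fields and player proximity.
# Layers length-preserving whole-plant bend with per-leaf rotation around pivots decoded from UV1/UV2.
# Wind actor data supplies intensity and direction, blended with physics and player push.
# Offsets fade by pixel depth, with static switches for camera-relative math, camera push-down and disabling animation; extra outputs expose pivots, masks and randoms.
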MaterialFunction /WildEstate/Foliage/BaseMaterial/MF_WildEstate_MediumPlantAnimation
graph {nodes [
  n1 = NamedRerouteUsage()
  n2 = Clamp(None=LinearInterpolate(A=ScalarParameter() {defaultValue=0.5, group=Animation Medium Plant 2024, parameterName=Overall_Bend_Falloff_min}, Alpha=VertexColor().A)) {clampMode=CMODE_ClampMin}
  n3 = ScalarParameter() {defaultValue=5, group=Animation Medium Plant 2024, parameterName=Overall_Bend_Falloff_Exponent}
  n4 = Multiply(
   A=Multiply(A=NamedRerouteUsage(), B=Power(Base=@n2, Exp=@n3))
   B=ScalarParameter() {defaultValue=77.4633, group=Animation Medium Plant 2024, parameterName=Overall_Bend_Strength}
  )
  n5 = Add(
   A=StaticSwitchParameter(True=WorldPosition() {worldPositionShaderOffset=WPT_CameraRelativeNoOffsets}.XYZ, False=LocalPosition() {includedOffsets=ExcludeOffsets}.XYZ) {parameterName=UseCameraRelativeOptimization}
   B=NamedRerouteUsage()
  )
  n6 = StaticSwitchParameter(True=NamedRerouteUsage(), False=Constant3Vector()) {parameterName=UseCameraRelativeOptimization}
  n7 = Subtract(A=Add(A=@n5, B=Multiply(A=@n1, B=@n4)), B=@n6)
  n8 = Divide(A=Length(None=Subtract(A=@n5, B=@n6)) {is=class MaterialExpressionLength}, B=Length(None=@n7))
  n9 = Add(A=Subtract(A=Multiply(A=@n7, B=@n8), B=@n5), B=@n6)
  n10 = Normalize(VectorInput=CrossProduct(A=NamedRerouteUsage(), B=NamedRerouteUsage()))
  n11 = NamedRerouteUsage()
  n12 = ScalarParameter() {defaultValue=0.288, group=Animation Medium Plant 2024, parameterName=Leaf_Rotation_Max}
  n13 = Multiply(
   A=Add(A=ScalarParameter() {group=Animation Medium Plant 2024, parameterName=Leaf_Rotation_Bias}, B=Multiply(A=@n11, B=@n12))
   B=NamedRerouteUsage()
  )
  n14 = StaticSwitchParameter(True=NamedRerouteUsage(), False=NamedRerouteUsage()) {parameterName=UseCameraRelativeOptimization}
  n15 = StaticSwitchParameter(True=WorldPosition() {worldPositionShaderOffset=WPT_CameraRelativeNoOffsets}.XYZ, False=WorldPosition().XYZ) {parameterName=UseCameraRelativeOptimization}
  Add(
   A=@n9
   B=NamedRerouteDeclaration(None=RotateAboutAxis(NormalizedRotationAxis=@n10, RotationAngle=@n13, PivotPoint=@n14, Position=@n15)) {name=LeafRotationDelta_2024, nodeColor=<rgba 1 0.8906 0 1>}
  )
  Constant3Vector()
  n16 = FunctionInput() {inputType=FunctionInput_MaterialAttributes}
  n17 = GetMaterialAttributes(None=@n16) {attributeGetTypes [#1]}
  n18 = NamedRerouteUsage()
  n19 = Subtract(A=PixelDepth().R, B=ScalarParameter() {defaultValue=3500, parameterName=WPO_Fade_Offset})
  n20 = ScalarParameter() {defaultValue=1000, parameterName=WPO_Fade_Distance}
  n21 = LinearInterpolate(A=@n18, B=Constant3Vector(), Alpha=Saturate(None=Divide(A=@n19, B=@n20)))
  n22 = Add(A=@n17.WorldPositionOffset, B=@n21)
  n23 = GetMaterialAttributes(None=StaticSwitchParameter(
    True=@n16
    False=SetMaterialAttributes!(MaterialAttributes=@n17.MaterialAttributes, WorldPositionOffset=@n22) {attributeSetTypes [#1]}
   ) {parameterName=DisableAllAnimation}) {attributeGetTypes [#1]}
  n24 = MaterialFunctionCall(Color=@n23.WorldPositionOffset) {materialFunction=&SplitComponents}
  n25 = CameraPositionWS()
  n26 = NamedRerouteUsage()
  n27 = CollectionParameter() {collection=&FortniteMaterialParameters, parameterName=GrassPushDownRadius}
  n28 = Constant()
  n29 = CollectionParameter() {collection=&FortniteMaterialParameters, parameterName=IsPlayerSliding}
  n30 = LinearInterpolate(A=Multiply(A=SphereMask(A=@n25, B=@n26, Radius=@n27, Hardness=@n28), B=Constant() {r=-200}), B=Constant(), Alpha=@n29)
  n31 = Add(
   A=@n23.WorldPositionOffset
   B=MaterialFunctionCall(X=@n24.R, Y=@n24.R, Z=Add(A=@n24.B, B=@n30)) {materialFunction=&MakeFloat3}.Result
  )
  FunctionOutput(None=StaticSwitchParameter(
    True=SetMaterialAttributes(MaterialAttributes=@n23.MaterialAttributes, WorldPositionOffset=@n31) {attributeSetTypes [#1]}
    False=@n23.MaterialAttributes
   ) {parameterName=UseCameraPushDown})
  FunctionOutput(None=StaticSwitchParameter(True=Constant3Vector(), False=@n21) {parameterName=DisableAllAnimation}) {outputName=WPO Only, sortPriority=1}
  FunctionOutput(None=MaterialFunctionCall(Number=NamedRerouteUsage()) {materialFunction=&DebugScalarValues}.Result) {outputName=Debug, sortPriority=565}
  n32 = TextureCoordinate() {coordinateIndex=2}
  n33 = AppendVector(A=Frac(None=TextureCoordinate() {coordinateIndex=1}), B=ComponentMask(None=Frac(None=@n32)) {r=true})
  n34 = Constant3Vector() {constant=<rgba 0.5 0.5 0.5 0>}
  n35 = Subtract(A=Multiply(A=Subtract(A=@n33, B=@n34)) {constB=768}, B=Constant3Vector() {constant=<rgba 0 0 128 0>})
  n36 = TransformPosition(None=@n35) {transformSourceType=TRANSFORMPOSSOURCE_Instance, transformType=TRANSFORMPOSSOURCE_World}
  FunctionOutput(None=NamedRerouteDeclaration(None=@n36) {name=Pivot_2024_WorldSpace, nodeColor=<rgba 0 1 0.9531 1>}) {outputName=LeafPivot_WS, sortPriority=50}
  FunctionOutput(None=NamedRerouteDeclaration(None=@n35) {name=Pivot_2024_LocalSpace, nodeColor=<rgba 0 1 0.9531 1>}) {outputName=LeafPivot_LS, sortPriority=50}
  FunctionOutput(None=@n9) {outputName=Anim_BendOnly, sortPriority=50}
  n37 = ComponentMask(None=TextureCoordinate() {coordinateIndex=2}) {g=true}
  FunctionOutput(None=Saturate(None=Ceil(None=@n37))) {outputName=LeafMask, sortPriority=50}
  n38 = MaterialFunctionCall(Float3=@n36) {materialFunction=&BreakOutFloat3Components}
  n39 = Add(A=Add(A=@n38.R, B=@n38.R), B=@n38.B)
  FunctionOutput(None=Divide(A=@n39) {constB=33.33}) {outputName=RandomPerLeaf, sortPriority=100}
  FunctionOutput(None=Divide(A=@n39) {constB=31.1}) {outputName=RandomPerLeaf_2, sortPriority=102}
  n40 = Constant3Vector()
  NamedRerouteDeclaration(None=TransformPosition(None=@n40) {transformSourceType=TRANSFORMPOSSOURCE_Instance, transformType=TRANSFORMPOSSOURCE_World}) {name=BasePivot, nodeColor=<rgba 1 0 0.7031 1>}
  NamedRerouteDeclaration(None=StaticBoolParameter() {defaultValue=true, parameterName=UsePlayerInteraction}) {name=UsePlayerInteraction, nodeColor=<rgba 0.3359 0 1 1>}
  n41 = Multiply(A=VertexColor()) {constB=2}
  NamedRerouteDeclaration(None=Transform(None=Subtract(A=@n41)) {transformSourceType=TRANSFORMSOURCE_Instance}) {name=LeafDirectionVector, nodeColor=<rgba 1 0 0.9141 1>}
  n42 = MaterialFunctionCall!(
   Position=WorldPosition().XYZ
   IsTrees
   IsGrass
   IsMediumPlant=StaticBool() {value=true}
   IsMediumPlant_2024Anim
   IsMediumPlant2ndLevel
   MediumPlant_BasePivotPostion
  ) {materialFunction=/Game/Environments/Asteria/Foliage/Wind/MF_WindActorData}
  n43 = Multiply(A=ScalarParameter() {defaultValue=50, group=AnimationTweakingMobile, parameterName=Leaf_Anim_Mobile}, B=@n42.WindIntensitySingleSample)
  n44 = LinearInterpolate(
   A=ScalarParameter() {defaultValue=-1, group=AnimationTweakingMobile, parameterName=Leaf_Anim_Falloff_base}
   B=ScalarParameter() {defaultValue=1.5, group=AnimationTweakingMobile, parameterName=Leaf_Anim_Falloff_tip}
   Alpha=VertexColor().A
  )
  NamedRerouteDeclaration(None=Multiply(A=Multiply(A=@n43, B=@n42.WindI_Direction_WithNoise), B=Saturate(None=@n44))) {name=Simple Mobile Anim, nodeColor=<rgba 0 0.9297 1 1>}
  n45 = StaticSwitch(
   True=Multiply(A=NamedRerouteUsage(), B=ScalarParameter() {defaultValue=1, group=Animation, parameterName=Leaf_WindIntesity_Strength})
   False=Constant()
   Value=NamedRerouteUsage()
  )
  n46 = NamedRerouteUsage()
  n47 = ScalarParameter() {defaultValue=1, group=Animation Physics, parameterName=Leaf_PhysicsFieldOffsetStrength}
  n48 = StaticSwitch(True=LinearInterpolate(A=@n45, B=Multiply(A=@n46, B=@n47), Alpha=NamedRerouteUsage()), False=@n45, Value=NamedRerouteUsage())
  n49 = ScalarParameter() {defaultValue=0.5, group=Animation Player Interaction, parameterName=Leaf_PlayerOffsetStrength}
  n50 = NamedRerouteUsage()
  NamedRerouteDeclaration(None=StaticSwitch(True=LinearInterpolate(A=@n48, B=@n49, Alpha=@n50), False=@n48, Value=NamedRerouteUsage())) {name=WindIntesity_Leaf_2024_WithPhys And Player, nodeColor=<rgba 0 1 0.2031 1>}
  n51 = NamedRerouteUsage()
  n52 = NamedRerouteUsage()
  n53 = Multiply(
   A=NamedRerouteUsage()
   B=ScalarParameter() {defaultValue=3, group=Animation Player Interaction, parameterName=RotationAxisTransitionHardness}
  ) {constB=500}
  n54 = StaticSwitch(
   True=LinearInterpolate(A=NamedRerouteUsage(), B=NamedRerouteUsage(), Alpha=Saturate(None=@n53))
   False=NamedRerouteUsage()
   Value=NamedRerouteUsage()
  )
  NamedRerouteDeclaration(None=StaticSwitch(True=LinearInterpolate(A=@n54, B=@n51, Alpha=@n52), False=@n54, Value=NamedRerouteUsage())) {name=WindDirection_Overall_withPhysics, nodeColor=<rgba 1 0.6328 0 1>}
  n55 = StaticSwitch(
   True=Multiply(A=NamedRerouteUsage(), B=ScalarParameter() {defaultValue=1, group=Animation, parameterName=Overall_WindIntesity_Strength})
   False=Constant()
   Value=NamedRerouteUsage()
  )
  n56 = NamedRerouteUsage()
  n57 = ScalarParameter() {defaultValue=1, group=Animation Physics, parameterName=Overall_PhysicsFieldOffsetStrength}
  n58 = StaticSwitch(True=LinearInterpolate(A=@n55, B=Multiply(A=@n56, B=@n57), Alpha=NamedRerouteUsage()), False=@n55, Value=NamedRerouteUsage())
  n59 = ScalarParameter() {defaultValue=0.25, group=Animation Player Interaction, parameterName=Overall_PlayerOffsetStrength}
  n60 = NamedRerouteUsage()
  NamedRerouteDeclaration(None=StaticSwitch(True=LinearInterpolate(A=@n58, B=@n59, Alpha=@n60), False=@n58, Value=NamedRerouteUsage())) {name=WindIntesity_Overall_2024_WithPhys And Player, nodeColor=<rgba 1 0.2109 0 1>}
  NamedRerouteDeclaration(None=@n9) {name=Final Animation 2024, nodeColor=<rgba 0.5 1 0 1>}
  NamedRerouteDeclaration(None=TransformPosition(None=@n35) {transformSourceType=TRANSFORMPOSSOURCE_Instance, transformType=TRANSFORMPOSSOURCE_TranslatedWorld}) {name=Pivot_2024_CameraSpace, nodeColor=<rgba 0 1 0.9531 1>}
  n61 = NamedRerouteDeclaration(None=@n37) {name=Gradient_2024AlongLeaf, nodeColor=<rgba 1 0 0.7266 1>}
  n62 = ScalarParameter() {defaultValue=2, group=Animation Medium Plant 2024, parameterName=Leaf_Gradient_Falloff}
  NamedRerouteDeclaration(None=Saturate(None=Power(Base=@n61, Exp=@n62))) {name=Gradient_2024AlongLeaf_Modfied, nodeColor=<rgba 1 0 0.7266 1>}
  NamedRerouteDeclaration(None=ComponentMask(None=@n32) {g=true}) {name=Debug_Float, nodeColor=<rgba 0.8516 1 0 1>}
  n63 = NamedRerouteUsage()
  n64 = Add(
   A=Multiply(A=NamedRerouteUsage(), B=ScalarParameter() {defaultValue=5, group=Animation Medium Plant 2024, parameterName=Leaf_Rotation_Randomnes})
   B=Multiply(A=NamedRerouteUsage(), B=ScalarParameter() {defaultValue=20, group=Animation Medium Plant 2024, parameterName=Leaf_Rotation_DragWiggle})
  )
  NamedRerouteDeclaration(None=MaterialFunctionCall(Position=Add(A=@n63, B=@n64), IsMediumPlant=StaticBool() {value=true}, IsMediumPlant_2024Anim=StaticBool() {value=true}) {materialFunction=&MF_WindActorData}.WindIntensitySingleSample) {name=WindIntensity_Leaf_2024, nodeColor=<rgba 0 1 0.2031 1>}
  n65 = SamplePhysicsVectorField!(World Position=NamedRerouteUsage()) {fieldTarget=Vector_LinearVelocity}
  n66 = MaterialFunctionCall(Vector 3=Add(A=@n65, B=Constant3Vector() {constant=<rgba 0.0001 0.0001 0.0001 0>}) {constB=0.0001}) {materialFunction=/Engine/Functions/Engine_MaterialFunctions02/Utility/VectorLength}
  n67 = Constant()
  NamedRerouteDeclaration(None=ShadingPathSwitch(Default=Saturate(None=@n66.V3 Length), Mobile=@n67) {desc=Migrated from FeatureLevelSwitch}) {name=PhysicsFieldMask, nodeColor=<rgba 1 0.75 0 1>}
  n68 = Constant3Vector() {constant=<rgba 0 0 1 0>}
  NamedRerouteDeclaration(None=ShadingPathSwitch(
    Default=MaterialFunctionCall(Vector=@n65, Default=@n68) {materialFunction=&SafeNormalize}.Result
    Mobile=Constant3Vector() {constant=<rgba 0 0 1 0>}
   ) {desc=Migrated from FeatureLevelSwitch}) {name=PhysicsFieldDirection, nodeColor=<rgba 1 0.75 0 1>}
  n69 = Min(A=@n66.V3 Length, B=ScalarParameter() {defaultValue=1, group=PhysicsField, parameterName=MaxRotation_PhysicsField_Clamp})
  n70 = Time()
  n71 = VertexColor()
  n72 = Add(A=@n71.R, B=@n71.R)
  n73 = Multiply(
   A=ScalarParameter() {defaultValue=351, group=PhysicsField, parameterName=PhysicsField_NoiseOffset}
   B=Multiply(A=Add(A=@n72, B=@n71.B)) {constB=33.33}
  )
  n74 = Frac(None=Multiply(A=ScalarParameter() {defaultValue=2.5, group=PhysicsField, parameterName=MaxRotation_PhysicsField_NoiseSpeed}, B=Add(A=@n70, B=@n73)))
  n75 = OneMinus(None=@n74)
  n76 = LinearInterpolate(
   A=Multiply(A=Multiply(A=@n74, B=@n75)) {constB=2}
   Alpha=ScalarParameter() {defaultValue=0.6, group=PhysicsField, parameterName=MaxRotation_PhysicsField_MInBend}
  )
  NamedRerouteDeclaration(None=ShadingPathSwitch(Default=Multiply(A=@n69, B=@n76), Mobile=@n67) {desc=Migrated from FeatureLevelSwitch}) {name=PhysicsFieldStrength, nodeColor=<rgba 1 0.75 0 1>}
  n77 = MaterialFunctionCall(
   Position=NamedRerouteUsage()
   IsMediumPlant_2024Anim=StaticBool() {value=true}
   IsMediumPlant2ndLevel=StaticBool() {value=true}
   MediumPlant_BasePivotPostion=NamedRerouteUsage()
  ) {materialFunction=&MF_WindActorData}
  NamedRerouteDeclaration(None=@n77.WindIntensitySingleSample) {name=WindIntensity_Overall_2024, nodeColor=<rgba 0 1 0.5781 1>}
  NamedRerouteDeclaration(None=@n77.WindI_Direction) {name=Wind Direction Overall 2024, nodeColor=<rgba 1 0.2578 0 1>}
  NamedRerouteDeclaration(None=TransformPosition(None=@n40) {transformSourceType=TRANSFORMPOSSOURCE_Instance, transformType=TRANSFORMPOSSOURCE_TranslatedWorld}) {name=BasePivot__CameraRelative, nodeColor=<rgba 1 0 0.7031 1>}
  NamedRerouteDeclaration(None=StaticBool() {value=true}) {name=Use Physics, nodeColor=<rgba 1 0.75 0 1>}
  n78 = CollectionParameter() {collection=&FortniteMaterialParameters, parameterName=PlayerPosition}
  n79 = ComponentMask(None=StaticSwitchParameter(
    True=TransformPosition(None=@n78) {transformSourceType=TRANSFORMPOSSOURCE_World, transformType=TRANSFORMPOSSOURCE_TranslatedWorld}
    False=@n78
   ) {parameterName=UseCameraRelativeOptimization}) {b=true, g=true, r=true}
  n80 = StaticSwitchParameter(True=NamedRerouteUsage(), False=NamedRerouteUsage()) {parameterName=UseCameraRelativeOptimization}
  n81 = ScalarParameter() {defaultValue=200, group=Animation Physics, parameterName=MaxDistanceForPlayerInteraction}
  n82 = Constant()
  NamedRerouteDeclaration(None=Saturate(None=SphereMask(A=@n79, B=@n80, Radius=@n81, Hardness=@n82) {attenuationRadius=64, hardnessPercent=20})) {name=PlayerSphereMask, nodeColor=<rgba 0 1 0.75 1>}
  n83 = Constant3Vector() {constant=<rgba 0 0.001 0 0>}
  n84 = Multiply(A=Constant3Vector() {constant=<rgba 1 1 0 0>}, B=Subtract(A=@n80, B=@n79))
  NamedRerouteDeclaration(None=Normalize(VectorInput=Add(A=@n83, B=@n84))) {name=PlayerPushDirection, nodeColor=<rgba 0 1 0.75 1>}
  NamedRerouteDeclaration(None=StaticBoolParameter() {defaultValue=true, parameterName=UseWind}) {name=UseWind, nodeColor=<rgba 1 0.3196 0 1>}
  n85 = NamedRerouteUsage()
  ShadingPathSwitch(Default=QualitySwitch(Default=@n18, Low=@n85, High=@n18, Medium=@n85, Epic=@n18), Mobile=@n85) {desc=Migrated from FeatureLevelSwitch}
 ]}
