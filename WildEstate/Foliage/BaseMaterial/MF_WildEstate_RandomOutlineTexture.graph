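# Foliage material function that picks one of four outline mask variants per leaf from a single mask texture, for WildEstate materials.
# The G and B channels each pack two masks, unpacked from above and below 0.5 and inverted.
# RandomPerleaf selects among the four by quarter, using steep divide-and-saturate steps and Round as branchless switches.
# RandomPerleaf_2 at or below the Outline_ExcludeThreshold parameter outputs white, excluding that leaf from outlining.
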MaterialFunction /WildEstate/Foliage/BaseMaterial/MF_WildEstate_RandomOutlineTexture
graph {nodes [
  n1 = TextureSample(Tex=FunctionInput() {bUsePreviewValueAsDefault=true, inputType=FunctionInput_Texture2D}) {samplerType=SAMPLERTYPE_Masks, texture=/Game/Characters/Player/Female/Medium/Bodies/F_MED_Commando_01/Textures/SolidWhite_MaskCompression}
  n2 = Saturate(None=Subtract(A=@n1.G) {constB=0.5})
  n3 = Saturate(None=Subtract(A=OneMinus(None=@n1.G)) {constB=0.5})
  n4 = FunctionInput() {inputName=RandomPerleaf, inputType=FunctionInput_Scalar, sortPriority=5}
  n5 = Subtract(A=@n4) {constB=0.25}
  n6 = LinearInterpolate(A=OneMinus(None=Multiply(A=@n2) {constB=2}), B=OneMinus(None=Multiply(A=@n3) {constB=2}), Alpha=Saturate(None=Divide(A=@n5) {constB=0.0001}))
  n7 = Saturate(None=Subtract(A=@n1.B) {constB=0.5})
  n8 = Saturate(None=Subtract(A=OneMinus(None=@n1.B)) {constB=0.5})
  n9 = Subtract(A=@n4) {constB=0.75}
  n10 = LinearInterpolate(A=OneMinus(None=Multiply(A=@n7) {constB=2}), B=OneMinus(None=Multiply(A=@n8) {constB=2}), Alpha=Saturate(None=Divide(A=@n9) {constB=0.0001}))
  n11 = Round(None=@n4)
  n12 = Ceil(None=Subtract(
    A=FunctionInput() {inputName=RandomPerleaf_2, inputType=FunctionInput_Scalar, sortPriority=10}
    B=ScalarParameter() {defaultValue=0.125, group=Outlines, parameterName=Outline_ExcludeThreshold}
   ))
  FunctionOutput(None=LinearInterpolate(B=LinearInterpolate(A=@n6, B=@n10, Alpha=@n11), Alpha=Saturate(None=@n12)) {constA=1}) {outputName=Lerp}
 ]}
