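# Masked, two-sided master material for WildEstate stylized grass and bushes, built for instanced and Nanite foliage.
# Static switches layer bush hue variation, snow, outlines and drop shading over MF_WildEstate_Grass color, with distance fades.
# World position offset branches cover wind sway, camera-based bending, leaf billboarding and sinking at distance.
# Quantized position-based roughness groups bush outlines, and an optional Clear Coat shading model serves as a post-process mask before MF_Curie_Landscape.
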Material /WildEstate/Foliage/BaseMaterial/M_WildEstate_Grass
graph {
 properties {
  bAutomaticallySetUsageInEditor=false
  bTangentSpaceNormal=false
  bUseMaterialAttributes=true
  bUsedWithInstancedStaticMeshes=true
  bUsedWithNanite=true
  blendMode=BLEND_Masked
  maxWorldPositionOffsetDisplacement=500
  shadingModel=MSM_FromMaterialExpression
  twoSided=true
  usedShadingModels=Default Lit | Clear Coat
 }
 nodes [
  Add(A=Divide(A=PixelDepth().R) {constB=70000}) {constB=0.01}
  n1 = NamedRerouteUsage()
  n2 = Distance(A=@n1, B=CameraPositionWS())
  n3 = ScalarParameter() {defaultValue=5336.106, parameterName=CullFadeOffset}
  n4 = Saturate(None=Divide(A=Subtract(A=@n2, B=@n3), B=ScalarParameter() {defaultValue=20089.8906, parameterName=CullDistance}))
  Add(A=PerInstanceRandom(), B=@n4)
  n5 = WorldPosition() {worldPositionShaderOffset=WPT_ExcludeAllShaderOffsets}
  Add(A=NamedRerouteUsage(), B=@n5.XYZ)
  Constant3Vector()
  Constant3Vector() {constant=<rgba 1 1 0 0>}
  Constant3Vector()
  Constant() {r=0.02}
  n6 = TextureSample() {samplerType=SAMPLERTYPE_Normal, texture=/WildEstate/Foliage/Grass/SmallBushes/WildEstate_SmallBushes_N}
  DotProduct(A=Constant3Vector() {constant=<rgba 0 -1 0 0>}, B=@n6.RGB)
  n7 = WorldPosition()
  n8 = Add(A=NamedRerouteUsage(), B=Multiply(A=CameraVectorWS(), B=ScalarParameter() {defaultValue=34.4652, parameterName=NormalFade}))
  n9 = CollectionParameter() {
   collection=&FortniteMaterialParameters
   parameterName=SunlightDirectionalVector
  }
  n10 = ComponentMask(None=@n9) {b=true, g=true, r=true}
  DotProduct(A=Normalize(VectorInput=Subtract(A=@n7.XYZ, B=@n8)), B=@n10)
  n11 = ShadingPathSwitch(
   Default=ScalarParameter() {parameterName=Grass_Outline_Intensity}
   Mobile=ScalarParameter() {defaultValue=0.5, parameterName=Grass_Outline_Mobile_intensity}
  )
  n12 = TextureSampleParameter2D() {
   parameterName=OpacityMask
   samplerType=SAMPLERTYPE_Masks
   texture=/Game/Characters/Player/Female/Medium/Bodies/F_MED_Commando_01/Textures/SolidWhite_MaskCompression
   channelNames {a="", b="", g=Outlines For Grass, r=Opacity Mask}
  }
  n13 = Saturate(None=LinearInterpolate(A=@n11, Alpha=@n12.Outlines For Grass) {constA=-5})
  n14 = LinearInterpolate(B=@n13, Alpha=NamedRerouteUsage()) {constA=1}
  n15 = TextureObjectParameter() {
   parameterName=OpacityMask
   samplerType=SAMPLERTYPE_Masks
   texture=&SolidWhite_MaskCompression
  }
  n16 = Subtract(A=TextureProperty(Tex=@n15), B=ScalarParameter() {defaultValue=64, parameterName=MipSizeThreshold})
  LinearInterpolate(A=@n14, Alpha=Saturate(None=Ceil(None=@n16)))
  n17 = Subtract(A=@n5.XYZ, B=NamedRerouteUsage())
  n18 = Constant3Vector() {constant=<rgba 1 0 0 0>}
  n19 = Constant3Vector() {constant=<rgba 0 1 0 0>}
  n20 = Constant3Vector() {constant=<rgba 0 0 1 0>}
  MaterialFunctionCall(
   VectorToTransform=@n17
   BasisX=Normalize(VectorInput=Transform(None=@n18) {transformSourceType=TRANSFORMSOURCE_Instance})
   BasisY=Normalize(VectorInput=Transform(None=@n19) {transformSourceType=TRANSFORMSOURCE_Instance})
   BasisZ=Normalize(VectorInput=Transform(None=@n20) {transformSourceType=TRANSFORMSOURCE_Instance})
  ) {materialFunction=&Transform3x3Matrix}
  n21 = Normalize(VectorInput=VertexNormalWS())
  n22 = Normalize(VectorInput=Transform(None=Constant3Vector() {constant=<rgba 0 0 1 0>}) {transformSourceType=TRANSFORMSOURCE_Instance})
  n23 = Normalize(VectorInput=CrossProduct(A=@n21, B=@n22))
  MaterialFunctionCall!(VectorToTransform=@n17, BasisX=@n21, BasisY=@n23, BasisZ=@n22) {materialFunction=/Engine/Functions/Engine_MaterialFunctions02/Math/InverseTransformMatrix}
  MaterialFunctionCall(VectorToTransform=@n17, BasisX=@n21, BasisY=@n23, BasisZ=@n22) {materialFunction=&Transform3x3Matrix}
  n24 = ShadingModel() {shadingModel=MSM_ClearCoat}
  n25 = Constant()
  n26 = Normalize(VectorInput=Transform(None=Constant3Vector() {constant=<rgba 0 0 1 0>}) {transformSourceType=TRANSFORMSOURCE_Instance})
  n27 = StaticBool()
  n28 = MaterialFunctionCall!(NormalOverrideWS=PixelNormalWS(), UseCamera facing BildBoards=@n27, UseLandscapeCoords=@n27) {materialFunction=/WildEstate/Environment/BaseMaterials/MaterialFunctions/MF_WildEstate_Grass}
  n29 = MaterialFunctionCall(
   Hue Shift Percentage=LinearInterpolate(
    A=ScalarParameter() {defaultValue=-0.144, group=Color, parameterName=Bushes ColorTint Var Low, sortPriority=51}
    B=ScalarParameter() {defaultValue=0.064, group=Color, parameterName=Bushes ColorTint Var High, sortPriority=51}
    Alpha=NamedRerouteUsage()
   )
   Texture=Multiply(A=VectorParameter() {parameterName=BushColorTint, defaultValue=<rgba 0 1.25 0.1578 1>}.RGB, B=@n28.Grass BaseColor)
  ) {materialFunction=&HueShift}
  n30 = StaticSwitchParameter(
   True=MaterialFunctionCall!(In Material, NormalOverrideWS=@n26, LayerWeight, ArrayTextureUV) {materialFunction=/WildEstate/Landscape/M_WildEstate_Snow}.BaseColor
   False=StaticSwitchParameter(True=@n29.Result, False=@n28.Grass BaseColor) {group=Color, parameterName=UseBushColor, sortPriority=50}
  ) {parameterName=isSnow}
  n31 = StaticSwitchParameter(True=Multiply(A=NamedRerouteUsage(), B=@n30), False=@n30) {parameterName=UseOutlineTexture}
  n32 = StaticSwitchParameter(True=Multiply(A=NamedRerouteUsage(), B=@n31), False=@n31) {parameterName=UseOutlines Grass}
  n33 = StaticSwitchParameter(True=Multiply(A=NamedRerouteUsage(), B=@n32), False=@n32) {parameterName=Use DropShadow Shading}
  n34 = Subtract(A=PixelDepth().R, B=ScalarParameter() {defaultValue=4000, parameterName=ColorFadeDistance_Offset})
  n35 = ScalarParameter() {defaultValue=3000, parameterName=ColorFadeDistance_Blend}
  n36 = NamedRerouteDeclaration(None=Saturate(None=Divide(A=@n34, B=@n35))) {name=FadePixelDepth, nodeColor=<rgba 0.0312 0 1 1>}
  n37 = NamedRerouteUsage()
  n38 = NamedRerouteUsage()
  n39 = NamedRerouteUsage()
  n40 = StaticSwitchParameter(True=NamedRerouteUsage(), False=NamedRerouteUsage()) {parameterName=UseBushesBillboard}
  n41 = SetMaterialAttributes!(
   MaterialAttributes
   Base Color=StaticSwitchParameter(True=LinearInterpolate(A=@n33, B=@n28.Grass BaseColor, Alpha=@n36), False=@n33) {group=Color, parameterName=UseBushColor, sortPriority=50}
   Specular=ScalarParameter() {parameterName=Specular}
   Roughness=StaticSwitchParameter(
    True=NamedRerouteUsage()
    False=MaterialFunctionCall() {materialFunction=&/MF_WildEstate_GrassRoughness}.Result
   ) {parameterName=Use BushRoughnessOutlines}
   Normal=StaticSwitchParameter(True=LinearInterpolate(A=@n37, B=@n26, Alpha=@n38), False=@n26) {parameterName=Use BushNormals}
   World Position Offset=StaticSwitchParameter(True=Add(A=@n39, B=@n40), False=@n40) {parameterName=UseWPOFade}
   Opacity Mask=NamedRerouteUsage()
   Opacity=Constant()
   Subsurface Color=Constant3Vector()
  ) {attributeSetTypes [
    #1
    #2
    #3
    #4
    #5
    #6
    #7
    #8
   ]}
  n42 = MaterialFunctionCall!(In=@n41, Enabled, Disabled, useDirectionalShading, DirectionalShadingMask) {materialFunction=/WildEstate/Environment/BaseMaterials/MaterialFunctions/MF_Test5050Unlit}
  n43 = ShadingModel()
  MaterialFunctionCall!(
   In=StaticSwitchParameter(
    True=SetMaterialAttributes!(MaterialAttributes=@n42.Result, Shading Model=@n24, Clear Coat or Custom Data 0=@n25) {attributeSetTypes [#9, #10]}
    False=SetMaterialAttributes!(MaterialAttributes=@n42.Result, Shading Model=@n43) {attributeSetTypes [#9]}
   ) {group=CellShading, parameterName=UsePropMaskForPost}
   UseHeightFalloff=StaticBool()
   OnFireEmissiveOverride
   HeightFalloffRadius
   HeightFalloffHardness
   VariationMasksMobile
   In_Grass=StaticBool() {value=true}
   Is_Stylised_WE
  ) {materialFunction=/Game/Packages/Fortress_SharedMaterials/Base_Material_Functions/MF_Curie_Landscape}
  n44 = NamedRerouteUsage()
  n45 = WorldPosition() {worldPositionShaderOffset=WPT_ExcludeAllShaderOffsets}
  n46 = TransformPosition(None=Constant3Vector()) {transformType=TRANSFORMPOSSOURCE_World}
  n47 = Constant3Vector() {constant=<rgba 1 0 0 0>}
  n48 = Constant3Vector() {constant=<rgba 0 1 0 0>}
  n49 = Constant3Vector() {constant=<rgba 0 0 1 0>}
  MaterialFunctionCall(
   VectorToTransform=Subtract(A=Add(A=@n44, B=@n45.XYZ), B=NamedRerouteDeclaration(None=@n46) {name=InstancePivot_1, nodeColor=<rgba 0 0.4375 1 1>})
   BasisX=Normalize(VectorInput=Transform(None=@n47) {transformSourceType=TRANSFORMSOURCE_World, transformType=TRANSFORM_Instance})
   BasisY=Normalize(VectorInput=Transform(None=@n48) {transformSourceType=TRANSFORMSOURCE_World, transformType=TRANSFORM_Instance})
   BasisZ=Normalize(VectorInput=Transform(None=@n49) {transformSourceType=TRANSFORMSOURCE_Instance})
  ) {materialFunction=&Transform3x3Matrix}
  MaterialFunctionCall!(NoRVT=@n26, RVT=NamedRerouteUsage()) {materialFunction=/Game/Packages/Fortress_SharedMaterials/Base_Material_Functions/FN_RVT_Switch_Color}
  n50 = TextureSampleParameter2D(UVs=Multiply(A=TextureCoordinate(), B=ScalarParameter() {defaultValue=1, parameterName=PDO_Tiling})) {
   group=PDO
   parameterName=Height
   samplerType=SAMPLERTYPE_LinearGrayscale
   texture=&T_SimplePerlin_Greyscale
  }
  n51 = ScalarParameter() {defaultValue=5, parameterName=PDO_Strength_High}
  Multiply(A=@n50.G, B=@n51)
  n52 = CameraVectorWS()
  n53 = Transform(None=Constant3Vector() {constant=<rgba 0 0 -1 0>}) {transformSourceType=TRANSFORMSOURCE_Camera}
  n54 = CrossProduct(A=@n52, B=@n53)
  n55 = Normalize(VectorInput=MaterialFunctionCall(VectorToTransform=@n6.RGB, BasisX=@n52, BasisY=Normalize(VectorInput=@n54), BasisZ=Normalize(VectorInput=@n53)) {materialFunction=&Transform3x3Matrix}.Transformed Vector)
  n56 = Ceil(None=Add(A=DotProduct(A=@n10, B=@n55), B=ScalarParameter() {defaultValue=0.2601, parameterName=ShadowBias}))
  Multiply(A=LinearInterpolate(A=ScalarParameter() {parameterName=ShadowDarkness}, Alpha=Saturate(None=@n56)), B=@n29.Result)
  n57 = TextureObjectParameter() {
   group=Textures
   parameterName=BushOutlineTexture
   samplerType=SAMPLERTYPE_Masks
   texture=&SolidWhite_MaskCompression
  }
  n58 = NamedRerouteUsage()
  n59 = NamedRerouteUsage()
  NamedRerouteDeclaration(None=LinearInterpolate(
    B=MaterialFunctionCall!(In=@n57, RandomPerleaf=@n58, RandomPerleaf_2=@n59) {materialFunction=&/MF_WildEstate_RandomOutlineTexture}.Lerp
    Alpha=NamedRerouteUsage()
   ) {constA=1}) {name=Outlines, nodeColor=<rgba 1 0 0.0469 1>}
  n60 = CollectionParameter() {
   collection=&FortniteMaterialParameters
   parameterName=SunlightDirectionalVector
  }
  n61 = ComponentMask(None=@n60) {b=true, g=true, r=true}
  n62 = CameraVectorWS()
  NamedRerouteDeclaration(None=Normalize(VectorInput=Add(A=@n61, B=@n62) {desc=Compromise for Direction light and all othe rlights})) {name=NormalsFlat, nodeColor=<rgba 0 1 0.4609 1>}
  n63 = Normalize(VectorInput=Transform(None=Constant3Vector() {constant=<rgba 1 0 0 0>}) {transformSourceType=TRANSFORMSOURCE_Camera})
  n64 = Normalize(VectorInput=Transform(None=Constant3Vector() {constant=<rgba 0 0 1 0>}) {transformSourceType=TRANSFORMSOURCE_Instance})
  n65 = NamedRerouteUsage()
  n66 = Subtract(A=CameraPositionWS(), B=@n65)
  n67 = Normalize(VectorInput=CrossProduct(A=@n63, B=Normalize(VectorInput=@n66)))
  n68 = Constant() {r=0.5}
  n69 = Multiply(
   A=Arccosine!(None=DotProduct(A=@n64, B=@n67))
   B=Divide(A=MaterialFunctionCall(Multiplier=@n68) {materialFunction=&Pi}.Result) {constB=10}
  )
  n70 = Constant3Vector() {constant=<rgba 1 1 0 0>}
  n71 = Multiply(A=@n70, B=Transform(None=@n66) {transformSourceType=TRANSFORMSOURCE_World, transformType=TRANSFORM_Instance})
  n72 = CollectionParameter() {
   collection=&FortniteMaterialParameters
   parameterName=SunlightDirectionalVector
  }
  n73 = ShadowReplace(
   Default=Normalize(VectorInput=Transform(None=@n71) {transformSourceType=TRANSFORMSOURCE_Instance})
   Shadow=ComponentMask(None=@n72) {b=true, g=true, r=true}
  )
  n74 = Transform(None=VertexNormalWS()) {transformSourceType=TRANSFORMSOURCE_World, transformType=TRANSFORM_Instance}
  n75 = Normalize(VectorInput=Transform(None=Multiply(A=@n70, B=@n74)) {transformSourceType=TRANSFORMSOURCE_Instance})
  n76 = DotProduct(A=@n73, B=@n75)
  n77 = MaterialFunctionCall(Multiplier=Constant() {r=0.5}) {materialFunction=&Pi}
  n78 = RotateAboutAxis(
   NormalizedRotationAxis=Normalize(VectorInput=CrossProduct(A=@n75, B=@n73))
   RotationAngle=Multiply(A=Arccosine(None=@n76), B=Divide(A=@n77.Result) {constB=10})
   PivotPoint=@n65
   Position=WorldPosition().XYZ
  )
  n79 = Add(A=WorldPosition().XYZ, B=@n78)
  n80 = NamedRerouteDeclaration(None=Add(A=RotateAboutAxis(NormalizedRotationAxis=@n63, RotationAngle=@n69, PivotPoint=@n65, Position=@n79), B=@n78)) {name=BillboardTransformFallback, nodeColor=<rgba 1 0.1641 0 1>}
  n81 = Multiply(
   A=Transform(None=Constant3Vector() {constant=<rgba 0 0 1 0>}) {transformSourceType=TRANSFORMSOURCE_Camera}
   B=ScalarParameter() {defaultValue=-20, parameterName=Billboard_Push}
  )
  NamedRerouteDeclaration(None=Add(A=NamedRerouteUsage(), B=Add(A=@n80, B=@n81))) {name=WPO BillboardTransformFallback, nodeColor=<rgba 1 0.1641 0 1>}
  n82 = ScalarParameter() {defaultValue=1.0657, parameterName=DropHighlight_Brightness}
  n83 = NamedRerouteUsage()
  n84 = ScalarParameter() {defaultValue=0.316, parameterName=DropHighlight_RandomThreshold}
  n85 = Saturate(None=Ceil(None=Subtract(A=@n83, B=@n84)))
  n86 = ScalarParameter() {defaultValue=0.684, parameterName=DropShadow_MinDarkness}
  n87 = StaticSwitchParameter(True=LinearInterpolate(A=@n86, Alpha=NamedRerouteUsage()) {constB=0.75}, False=@n86) {parameterName=FadeOutRoughnesOutlines InDistance}
  n88 = TextureSampleParameter2D(UVs=TextureCoordinate()) {
   parameterName=OpacityMask
   samplerType=SAMPLERTYPE_Masks
   texture=&SolidWhite_MaskCompression
   channelNames {a="", b="", g=Outlines For Grass, r=Opacity Mask}
  }
  n89 = ScalarParameter() {defaultValue=0.336, parameterName=DropShadow_Threshold}
  n90 = Round(None=Saturate(None=Subtract(A=@n88.B, B=@n89)))
  n91 = Saturate(None=Subtract(A=@n88.Outlines For Grass, B=ScalarParameter() {defaultValue=0.272, parameterName=DropHighlight_Threshold}))
  NamedRerouteDeclaration(None=LinearInterpolate(A=LinearInterpolate(A=@n82, Alpha=@n85), B=LinearInterpolate(A=@n87, Alpha=@n90) {constA=1}, Alpha=Round(None=@n91)) {constA=1}) {name=DropShading, nodeColor=<rgba 1 0 0.0469 1>}
  n92 = MaterialFunctionCall!(In) {materialFunction=&/MF_WildEstate_MediumPlantAnimation}
  n93 = PerInstanceRandom()
  NamedRerouteDeclaration(None=Frac(None=Add(A=@n92.RandomPerLeaf, B=@n93))) {name=RandomPerleaf, nodeColor=<rgba 0.1484 1 0 1>}
  n94 = TransformPosition(None=Constant3Vector()) {transformSourceType=TRANSFORMPOSSOURCE_Instance, transformType=TRANSFORMPOSSOURCE_World}
  n95 = Divide(A=ComponentMask(None=@n94) {g=true, r=true}, B=ScalarParameter() {defaultValue=1000, parameterName=BushRoughnessOutlineMergeSize})
  n96 = Frac(None=@n95)
  n97 = ScalarParameter() {defaultValue=4, parameterName=BushRoughnessOutlineMergeSteps}
  n98 = Divide(A=Round(None=Multiply(A=@n96, B=@n97)), B=@n97)
  n99 = LinearInterpolate(A=Constant() {r=0.05}, B=Constant() {r=0.49}, Alpha=@n98)
  n100 = Subtract(A=Distance(A=CameraPositionWS(), B=@n94), B=ScalarParameter() {defaultValue=7081.9189, parameterName=RoughnessFadeDistance_Offset})
  n101 = NamedRerouteDeclaration(None=Saturate(None=Ceil(None=@n100))) {name=MaskRoughnesOutlines, nodeColor=<rgba 0.8047 1 0 1>}
  NamedRerouteDeclaration(None=StaticSwitchParameter(True=LinearInterpolate(A=@n99, Alpha=@n101), False=@n99) {parameterName=FadeOutRoughnesOutlines InDistance}) {name=BushRoughness, nodeColor=<rgba 0.8984 1 0 1>}
  NamedRerouteDeclaration(None=@n98) {name=RandomGrouping, nodeColor=<rgba 0 1 0.1094 1>}
  n102 = MaterialFunctionCall!(
   Position=NamedRerouteUsage()
   IsTrees=StaticBool() {value=true}
   IsGrass
   IsMediumPlant
   IsMediumPlant_2024Anim=StaticBool() {value=true}
   IsMediumPlant2ndLevel
   MediumPlant_BasePivotPostion
  ) {materialFunction=/Game/Environments/Asteria/Foliage/Wind/MF_WindActorData}
  NamedRerouteDeclaration(None=@n102.WindIntensitySingleSample) {name=WindIntensity, nodeColor=<rgba 0 1 0.3438 1>}
  NamedRerouteDeclaration(None=Frac(None=Add(A=@n92.RandomPerLeaf_2, B=@n93))) {name=RandomPerleaf_2, nodeColor=<rgba 0.1484 1 0 1>}
  n103 = Divide(
   A=Subtract(A=PixelDepth().R, B=ScalarParameter() {defaultValue=4000, parameterName=WPOGroudnSInk_Offset})
   B=ScalarParameter() {defaultValue=3000, parameterName=WPOGroudnSInk_Blend}
  )
  NamedRerouteDeclaration(None=LinearInterpolate(A=Constant3Vector(), B=VectorParameter() {parameterName=WPOGrounSink, defaultValue=<rgba 0 0 -50 1>}.RGB, Alpha=Saturate(None=@n103))) {name=WPOGroundSink, nodeColor=<rgba 0 1 0.4375 1>}
  NamedRerouteDeclaration(None=@n92.LeafPivot_WS) {name=LeafPivot WorldSpace, nodeColor=<rgba 0.9688 0 1 1>}
  n104 = Saturate(None=LinearInterpolate(A=@n11, Alpha=@n12.B) {constA=-5})
  n105 = NamedRerouteUsage()
  NamedRerouteDeclaration(None=StaticSwitchParameter(True=LinearInterpolate(A=@n104, B=@n13, Alpha=@n105), False=@n14) {defaultValue=true, parameterName=UseDualOutlineTickness}) {name=Grass_Outlines, nodeColor=<rgba 1 0 0.0469 1>}
  n106 = Subtract(A=PixelDepth().R, B=ScalarParameter() {defaultValue=500, parameterName=OutlineNearFade_Offset})
  n107 = ScalarParameter() {defaultValue=500, parameterName=OutlineNearFade_Blend}
  NamedRerouteDeclaration(None=Saturate(None=Divide(A=@n106, B=@n107))) {name=NearOutlineFade, nodeColor=<rgba 0 0.2969 1 1>}
  n108 = Add(A=VectorParameter() {parameterName=CullFade_PushBelowGround, defaultValue=<rgba 0 0 -50 1>}.RGB, B=Subtract(A=@n1, B=WorldPosition().XYZ))
  n109 = Divide(A=LocalPosition().Z, B=ScalarParameter() {defaultValue=6.5821, parameterName=BendGradient})
  n110 = Constant3Vector()
  n111 = Subtract(
   A=Add(A=NamedRerouteUsage(), B=WorldPosition() {worldPositionShaderOffset=WPT_ExcludeAllShaderOffsets}.XYZ)
   B=NamedRerouteDeclaration(None=TransformPosition(None=@n110) {transformSourceType=TRANSFORMPOSSOURCE_Instance, transformType=TRANSFORMPOSSOURCE_World}) {name=InstancePivot, nodeColor=<rgba 0 0.4375 1 1>}
  )
  n112 = Normalize(VectorInput=Transform(None=Constant3Vector() {constant=<rgba 1 0 0 0>}) {transformSourceType=TRANSFORMSOURCE_Instance})
  n113 = Normalize(VectorInput=Transform(None=Constant3Vector() {constant=<rgba 0 1 0 0>}) {transformSourceType=TRANSFORMSOURCE_Instance})
  n114 = Normalize(VectorInput=Transform(None=Constant3Vector() {constant=<rgba 0 0 1 0>}) {transformSourceType=TRANSFORMSOURCE_Instance})
  n115 = Constant3Vector() {constant=<rgba 1 1 0 0>}
  n116 = Subtract(
   A=TransformPosition(None=Constant3Vector()) {transformSourceType=TRANSFORMPOSSOURCE_Instance, transformType=TRANSFORMPOSSOURCE_World}
   B=CameraPositionWS()
  )
  n117 = Normalize(VectorInput=Transform(None=Multiply(A=@n115, B=@n116)) {transformSourceType=TRANSFORMSOURCE_Instance})
  n118 = Normalize(VectorInput=LinearInterpolate(
    A=Constant3Vector() {constant=<rgba 0 0 1 0>}
    B=Normalize(VectorInput=@n116)
    Alpha=ScalarParameter() {defaultValue=0.42, parameterName=Bendaway}
   ))
  n119 = Transform(None=CrossProduct(A=@n118, B=@n117)) {transformSourceType=TRANSFORMSOURCE_Instance}
  n120 = Subtract(
   A=MaterialFunctionCall(
    VectorToTransform=MaterialFunctionCall(VectorToTransform=@n111, BasisX=@n112, BasisY=@n113, BasisZ=@n114) {materialFunction=&Transform3x3Matrix}.Transformed Vector
    BasisX=@n117
    BasisY=Normalize(VectorInput=@n119)
    BasisZ=@n118
   ) {materialFunction=&Transform3x3Matrix}.Transformed Vector
   B=@n111
  )
  n121 = Add(A=WorldPosition().XYZ, B=@n120)
  n122 = CameraPositionWS()
  n123 = Normalize(VectorInput=Multiply(A=Subtract(A=@n121, B=@n122), B=Constant3Vector() {constant=<rgba 1 1 0 0>}))
  n124 = Add(A=Multiply(A=Multiply(A=@n109, B=@n123), B=ScalarParameter() {defaultValue=-4.0319, parameterName=BendPush}), B=@n120)
  NamedRerouteDeclaration(None=LinearInterpolate(A=@n124, B=Add(A=@n124, B=@n108), Alpha=Saturate(None=@n4))) {name=WPO, nodeColor=<rgba 0.6406 1 0 1>}
  NamedRerouteDeclaration(None=@n12.Opacity Mask) {name=OpacityMask, nodeColor=<rgba 1 0 0.0469 1>}
  n125 = PerInstanceRandom()
  n126 = Add(A=Time(), B=@n125)
  n127 = ScalarParameter() {defaultValue=0.5257, group=Animation, parameterName=Anim_Speed}
  n128 = Multiply(A=@n126, B=@n127)
  n129 = ScalarParameter() {defaultValue=2, group=Animation, parameterName=Anim_Speed_Storm}
  n130 = Multiply(A=@n126, B=@n129)
  n131 = Multiply(A=@n125) {constB=0.01}
  n132 = Multiply(A=@n126, B=Add(A=@n131, B=@n127))
  n133 = Multiply(A=@n126, B=Add(A=@n131, B=@n129))
  n134 = Add(
   A=LinearInterpolate(A=Sine(None=@n128), B=Sine(None=@n130), Alpha=NamedRerouteUsage())
   B=LinearInterpolate(A=Sine(None=@n132), B=Sine(None=@n133), Alpha=NamedRerouteUsage())
  )
  n135 = Constant3Vector() {constant=<rgba 1 1 0 0>}
  n136 = ScalarParameter() {defaultValue=1, group=Animation, parameterName=Wind_Intensity}
  n137 = ScalarParameter() {defaultValue=50, group=Animation, parameterName=Wind_Intensity_Storm}
  n138 = NamedRerouteUsage()
  n139 = Divide(
   A=Subtract(
    A=LocalPosition() {localOrigin=InstancePreSkinning}.Z
    B=ScalarParameter() {defaultValue=3.5322, group=Animation, parameterName=Gradient_Bias}
   )
   B=ScalarParameter() {defaultValue=16.6253, group=Animation, parameterName=Gradient_Dist}
  )
  n140 = ScalarParameter() {defaultValue=1.5, group=Animation, parameterName=Gradient_Falloff}
  n141 = Multiply(A=Multiply(A=Multiply(A=@n134, B=@n135), B=LinearInterpolate(A=@n136, B=@n137, Alpha=@n138)), B=Saturate(None=Power(Base=@n139, Exp=@n140)))
  NamedRerouteDeclaration(None=@n141) {name=GrassAnim, nodeColor=<rgba 0.9688 1 0 1>}
  NamedRerouteDeclaration(None=@n102.StormMask) {name=StormMask, nodeColor=<rgba 0.7344 1 0 1>}
  NamedRerouteDeclaration(None=@n102.WindI_Direction_WithNoise) {name=WindIntensity_Direction, nodeColor=<rgba 0 1 0.3438 1>}
  n142 = RuntimeVirtualTextureSampleParameter(World Position=WorldPosition().XYZ) {
   bEnableFeedback=false
   materialType=BaseColor_Normal_Specular_YCoCg
   parameterName=RVT_Base
   virtualTexture=/Game/Athena/Apollo/Maps/Landscape/VT/RVT_Apollo_Landscape
  }
  NamedRerouteDeclaration(None=Multiply(A=@n142.BaseColor, B=Constant() {desc=Fix RVT MissMatch, r=1.02})) {name=RVT_BaseColor, nodeColor=<rgba 0 0.7656 1 1>}
  NamedRerouteDeclaration(None=Transform(None=NamedRerouteDeclaration(None=@n142.Normal) {name=RVT_Normal_WS, nodeColor=<rgba 0 0.7656 1 1>}) {transformSourceType=TRANSFORMSOURCE_World, transformType=TRANSFORM_Tangent}) {name=RVT_Normal_TS, nodeColor=<rgba 0 0.7656 1 1>}
  Normalize(VectorInput=Transform(None=Constant3Vector() {constant=<rgba 1 0 0 0>}) {transformSourceType=TRANSFORMSOURCE_Instance})
  n143 = Constant3Vector() {constant=<rgba 0 0 1 0>}
  Normalize(VectorInput=@n143)
  n144 = Constant3Vector() {constant=<rgba 1 0 0 0>}
  Normalize(VectorInput=@n144)
  n145 = Constant3Vector() {constant=<rgba 0 1 0 0>}
  Normalize(VectorInput=@n145)
  n146 = Constant3Vector() {constant=<rgba 0 0 1 0>}
  n147 = TransformPosition(None=Constant3Vector()) {transformSourceType=TRANSFORMPOSSOURCE_Instance, transformType=TRANSFORMPOSSOURCE_World}
  n148 = CameraPositionWS()
  n149 = Normalize(VectorInput=Subtract(A=ComponentMask(None=@n147) {g=true, r=true}, B=ComponentMask(None=@n148) {g=true, r=true}))
  n150 = Arctangent2Fast(Y=ComponentMask(None=@n149) {g=true}, X=ComponentMask(None=@n149) {r=true})
  n151 = MaterialFunctionCall(Multiplier=Constant() {r=2}) {materialFunction=&Pi}
  RotateAboutAxis(
   NormalizedRotationAxis=Normalize(VectorInput=Transform(None=@n146) {transformSourceType=TRANSFORMSOURCE_Instance})
   RotationAngle=Add(A=ScalarParameter() {defaultValue=0.5, parameterName=Test}, B=Divide(A=@n150, B=@n151.Result)) {constA=0.25}
   PivotPoint=@n147
   Position=WorldPosition().XYZ
  )
  ScalarParameter() {defaultValue=1, parameterName=Roughness}
  ScalarParameter() {defaultValue=0.096, parameterName=testets}
  n152 = CameraPositionWS()
  n153 = CollectionParameter() {
   collection=&FortniteMaterialParameters
   parameterName=SunlightDirectionalVector
  }
  n154 = ShadowReplace(Default=Normalize(VectorInput=Subtract(A=@n152, B=@n92.LeafPivot_WS)), Shadow=ComponentMask(None=@n153) {b=true, g=true, r=true})
  n155 = Normalize(VectorInput=VertexNormalWS())
  n156 = DotProduct(A=@n154, B=@n155)
  n157 = StaticSwitchParameter(True=Multiply(A=@n154, B=Sign(None=@n156)), False=@n154) {defaultValue=true, group=Billboard, parameterName=BillBoardShortestPathRotation}
  n158 = CrossProduct(A=@n155, B=@n157)
  n159 = Arccosine(None=DotProduct(A=@n157, B=@n155))
  n160 = Divide(A=MaterialFunctionCall(Multiplier=Constant() {r=0.5}) {materialFunction=&Pi}.Result) {constB=10}
  ShadowReplace(
   Default=RotateAboutAxis(
    NormalizedRotationAxis=Normalize(VectorInput=@n158)
    RotationAngle=Multiply(A=@n159, B=@n160)
    PivotPoint=@n92.LeafPivot_WS
    Position=WorldPosition().XYZ
   )
   Shadow=Constant3Vector()
  )
  Sine(None=@n95)
  n161 = ScalarParameter() {defaultValue=5, parameterName=PDO_Strength_Low}
  n162 = Divide(A=LocalPosition().Z, B=ScalarParameter() {defaultValue=10, parameterName=PDO_GradientFromBottom})
  n163 = SetMaterialAttributes!(MaterialAttributes=@n41, Pixel Depth Offset=Multiply(A=LinearInterpolate(A=@n161, B=@n51, Alpha=@n50.G), B=Saturate(None=@n162))) {attributeSetTypes [#11]}
  StaticSwitchParameter(
   True=ShadingPathSwitch(Default=QualitySwitch(Default=@n163, Low=@n41, Medium=@n41), Mobile=@n41) {desc=Migrated from FeatureLevelSwitch}
   False=@n41
  ) {group=PDO, parameterName=Use PDO}
  n164 = DotProduct(A=@n73, B=@n75)
  StaticSwitchParameter(True=Multiply(A=@n73, B=Sign(None=@n164)), False=@n73) {defaultValue=true, group=Billboard, parameterName=BillBoardShortestPathRotation}
  Transform(None=@n141) {transformSourceType=TRANSFORMSOURCE_Local}
  Transform(None=Constant3Vector() {constant=<rgba 0 0 1 0>}) {transformSourceType=TRANSFORMSOURCE_Camera, transformType=TRANSFORM_Instance}
  n165 = Subtract(A=Transform(None=Constant3Vector()) {transformSourceType=TRANSFORMSOURCE_Instance}, B=CameraPositionWS())
  n166 = Constant3Vector() {constant=<rgba 1 1 0 0>}
  Transform(None=Normalize(VectorInput=Multiply(A=@n165, B=@n166))) {transformSourceType=TRANSFORMSOURCE_World, transformType=TRANSFORM_Instance}
  Transform(None=Constant3Vector() {constant=<rgba 0 0 1 0>}) {transformSourceType=TRANSFORMSOURCE_Camera}
  Transform(None=@n143) {transformSourceType=TRANSFORMSOURCE_Instance, transformType=TRANSFORM_Local}
  Transform(None=@n144) {transformSourceType=TRANSFORMSOURCE_Instance}
  Transform(None=@n145) {transformSourceType=TRANSFORMSOURCE_Instance}
 ]
}
