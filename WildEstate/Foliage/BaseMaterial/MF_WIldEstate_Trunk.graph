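# Material function that builds the stylized WildEstate tree trunk material, or passes input attributes through when IsTrunk is off.
# Tinted optional diffuse, a bark mask and an optional normal map feed color quantization and flat shading.
# A crack variant mixes in at a fixed, Nanite-dependent weight, then a grass-blend function takes landscape runtime virtual texture inputs.
# IsImpostorBake bypasses all post-processing and outputs the base attributes.
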MaterialFunction /WildEstate/Foliage/BaseMaterial/MF_WIldEstate_Trunk
graph {nodes [
  n1 = RuntimeVirtualTextureSampleParameter(World Position=WorldPosition().XYZ) {
   bEnableFeedback=false
   materialType=WorldHeight
   parameterName=RVT_WorldHeight
   virtualTexture=/Game/Athena/Apollo/Maps/Landscape/VT/RVT_Apollo_Landscape_Height
  }
  n2 = Subtract(A=WorldPosition().Z, B=NamedRerouteDeclaration(None=@n1.WorldHeight) {name=RVT_WordPosZ, nodeColor=<rgba 0 0.7656 1 1>})
  n3 = Saturate(None=Divide(
    A=NamedRerouteDeclaration(None=@n2) {name=RVT_WordPosZbiased, nodeColor=<rgba 0 0.7656 1 1>}
    B=ScalarParameter() {defaultValue=16, parameterName=BlendDistance}
   ))
  BlendMaterialAttributes(
   A=SetMaterialAttributes!(MaterialAttributes, EmissiveColor, BaseColor=Constant3Vector() {constant=<rgba 1 0 0.0213 1>}) {attributeSetTypes [#1, #2]}
   Alpha=@n3
  )
  n4 = VectorParameter() {group=Trunk, parameterName=Trunk_Color, defaultValue=<rgba 1 1 1 1>}
  n5 = Multiply(
   A=@n4.RGB
   B=TextureSampleParameter2D() {group=Trunk, parameterName=Trunk_Diffuse, texture=/Game/Global/Textures/Default/Blanks/T_MidgreysRGB}.RGB
  )
  n6 = TextureSampleParameter2D() {
   group=Trunk
   parameterName=OverlayTexture
   samplerType=SAMPLERTYPE_Masks
   texture=/WildEstate/Environment/TestContent/T_WildEstate_Tree_Bark2_D
  }
  n7 = TextureObjectParameter() {group=Trunk, parameterName=Normals, samplerType=SAMPLERTYPE_Normal, sortPriority=0, texture=/Engine/EngineMaterials/FlatNormal}
  n8 = StaticSwitchParameter(True=@n7, False=Constant3Vector() {constant=<rgba 0 0 1 0>}) {group=Trunk, parameterName=UseNoramlTexture}
  n9 = Transform(None=@n8)
  n10 = SetMaterialAttributes!(
   MaterialAttributes
   BaseColor=Multiply(A=StaticSwitchParameter(True=@n5, False=@n4.RGB) {group=Trunk, parameterName=UseDiffuseTexture}, B=@n6.R)
   Specular=Constant()
   Roughness=Constant() {r=1}
   Anisotropy
   EmissiveColor
   Opacity
   OpacityMask
   Normal=@n9
   SubsurfaceColor=NamedRerouteUsage()
  ) {attributeSetTypes [
    #2
    #3
    #4
    #5
    #1
    #6
    #7
    #8
    #9
   ]}
  n11 = MaterialFunctionCall!(
   Material In=MaterialFunctionCall!(In=@n10) {materialFunction=/WildEstate/Environment/BaseMaterials/MaterialFunctions/M_WildEstate_QantizeColor}.Result
   UseFlatShading=StaticBoolParameter() {group=FlatShading, parameterName=UseFlatShading_All, sortPriority=1}
   IsWorldSpace=StaticBool() {value=true}
   OverrideUnlitShadinBaseColorDarkening
   OverrideUnlitShadinBaseColorEmissive
  ) {materialFunction=/WildEstate/Environment/BaseMaterials/MaterialFunctions/MF_WildEstate_FlatShading}
  n12 = Constant() {r=0.675}
  n13 = Constant() {r=0.5}
  n14 = MaterialFunctionCall!(
   In=BlendMaterialAttributes(
    A=@n11.Result
    B=MaterialFunctionCall!(In=@n11.Result, In_Translucent, DestructibleDamageAmount) {materialFunction=/WildEstate/Environment/BaseMaterials/MaterialFunctions/MF_WildEstate_Cracks}.Result
    Alpha=NaniteReplace!(Default=@n12, Nanite=@n13)
   )
   RVT_Normal=NamedRerouteUsage()
   IsWorldSpace=StaticBool() {value=true}
   RVT_WorldPosZ=NamedRerouteUsage()
   RVT_BaseColor=NamedRerouteUsage()
   RVT_Specular=NamedRerouteUsage()
   RVT_Roughness=NamedRerouteUsage()
   L2_Mask
   L3_Mask
   L4_Mask
   IsTree=StaticBool() {value=true}
   UseLayerMasking=StaticBool()
  ) {materialFunction=/WildEstate/Environment/BaseMaterials/MaterialFunctions/M_WildEstate_GrassBlend}
  n15 = FunctionInput(Preview=StaticBool()) {bUsePreviewValueAsDefault=true, inputName=IsImpostorBake, inputType=FunctionInput_StaticBool, sortPriority=99}
  FunctionOutput(None=StaticSwitchParameter(True=StaticSwitch(True=@n10, False=@n14.Result, Value=@n15), False=FunctionInput() {inputType=FunctionInput_MaterialAttributes}) {group=Trunk, parameterName=IsTrunk})
  n16 = RuntimeVirtualTextureSampleParameter(World Position=WorldPosition().XYZ) {
   bEnableFeedback=false
   materialType=BaseColor_Normal_Specular_YCoCg
   parameterName=RVT_Base
   virtualTexture=/Game/Athena/Apollo/Maps/Landscape/VT/RVT_Apollo_Landscape
  }
  NamedRerouteDeclaration(None=Transform(None=NamedRerouteDeclaration(None=@n16.Normal) {name=RVT_Normal_WS, nodeColor=<rgba 0 0.7656 1 1>}) {transformSourceType=TRANSFORMSOURCE_World, transformType=TRANSFORM_Tangent}) {name=RVT_Normal_TS, nodeColor=<rgba 0 0.7656 1 1>}
  NamedRerouteDeclaration(None=Constant3Vector() {constant=<rgba 8 0.95 0 0>}) {name=Visual Warning 2, nodeColor=<rgba 1 0.6678 0 1>}
  NamedRerouteDeclaration(None=@n16.Roughness) {name=RVT_Roughness, nodeColor=<rgba 0 0.7656 1 1>}
  NamedRerouteDeclaration(None=@n16.Specular) {name=RVT_Specular, nodeColor=<rgba 0 0.7656 1 1>}
  NamedRerouteDeclaration(None=@n3) {name=RVT_HeightBlend, nodeColor=<rgba 0 0.7656 1 1>}
  NamedRerouteDeclaration(None=@n16.BaseColor) {name=RVT_BaseColor, nodeColor=<rgba 0 0.7656 1 1>}
  n17 = ComponentMask(None=WorldPosition().XYZ) {b=true}
  NamedRerouteDeclaration(None=Sine(None=Divide(A=@n17) {constB=50})) {name=VisualWaring 1, nodeColor=<rgba 1 0.7456 0 1>}
  StaticSwitch(True=@n9, False=@n8, Value=@n15)
 ]}
