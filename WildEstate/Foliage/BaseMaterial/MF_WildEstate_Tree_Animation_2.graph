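# Material function that adds wind-style World Position Offset animation to WildEstate tree foliage materials.
# It adds the soft-transform tree offset to incoming WPO, then runs Asteria's simple tree animation, with a static switch to bypass both.
# Extra outputs expose leaf pivot, the soft-transform offset and overall bend.
# A depth-based WPO fade, using a WPO-disable blend under Nanite, is built but feeds no output.
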MaterialFunction /WildEstate/Foliage/BaseMaterial/MF_WildEstate_Tree_Animation_2
graph {nodes [
  n1 = FunctionInput() {inputType=FunctionInput_MaterialAttributes}
  n2 = GetMaterialAttributes(None=@n1) {attributeGetTypes [#1, #2]}
  n3 = MaterialFunctionCall!(StormMask, OptionalWPO, TreeAddOnOverallBend) {materialFunction=&/MF_WPO_Tree_Soft_Transform_WildEstate}
  n4 = MaterialFunctionCall!(In=SetMaterialAttributes!(MaterialAttributes=@n2.MaterialAttributes, WorldPositionOffset=Add(A=@n2.WorldPositionOffset, B=@n3.Offset), Normal=@n2.MaterialAttributes) {attributeSetTypes [#2, #1]}) {materialFunction=/Game/Environments/Asteria/Foliage/Trees/MasterMaterials/MaterialFunctions/MF_Asteria_Tree_SimpleAnimation}
  FunctionOutput(None=StaticSwitchParameter(True=@n1, False=@n4.Result) {group=Animation, parameterName=DisableAllAnimation})
  FunctionOutput(None=@n3.LeafPivot_LocalSpace) {outputName=LeafPivot_LocalSpace, sortPriority=10}
  FunctionOutput(None=@n3.Offset) {outputName=Anim_WPO_Only, sortPriority=20}
  FunctionOutput(None=@n3.OveraAllBend) {outputName=OveraAllBend, sortPriority=100}
  n5 = GetMaterialAttributes(None=@n4.Result) {attributeGetTypes [#2]}
  n6 = Constant3Vector()
  n7 = Subtract(A=PixelDepth().R, B=ScalarParameter() {defaultValue=7500, group=WPO Fade, parameterName=WPO_Fade_Offset})
  n8 = ScalarParameter() {defaultValue=1500, group=WPO Fade, parameterName=WPO_Fade_Distance}
  n9 = Saturate(None=NaniteReplace!(
    Default=Divide(A=@n7, B=@n8)
    Nanite=MaterialFunctionCall!(FadeBlendDistance=@n8, AutoHandleBadDistance) {materialFunction=/Game/Global/MaterialFunctions/TechArt/MF_WPODisableBlend}.VertexShaderOutput
   ))
  StaticSwitchParameter(
   True=SetMaterialAttributes!(MaterialAttributes=@n5.MaterialAttributes, WorldPositionOffset=LinearInterpolate(A=@n5.WorldPositionOffset, B=@n6, Alpha=@n9)) {attributeSetTypes [#2]}
   False=@n4.Result
  ) {defaultValue=true, group=WPO Fade, parameterName=UseWPOFade}
 ]}
