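# Base material for WildEstate medium plants, a Nanite-enabled subsurface foliage shader with a leaf cutout mask and stylized lighting.
# The animation function supplies leaf pivots, which drive optional billboarding that turns leaves toward the camera, or toward the sun in shadow passes.
# Color combines trunk/leaf tints and hue variation with sun-driven stepped shading, switchable between leaf-pivot, virtual-plane heightmap and hedge modes.
# Final switches handle impostor baking, unlit output and distance-based WPO fade.
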Material /WildEstate/Foliage/BaseMaterial/M_WildEstate_MediumPlant_BaseMat
graph {
 properties {
  bAutomaticallySetUsageInEditor=false
  bHasPixelAnimation=true
  bTangentSpaceNormal=false
  bUseMaterialAttributes=true
  bUsedWithInstancedStaticMeshes=true
  bUsedWithNanite=true
  shadingModel=MSM_Subsurface
  usedShadingModels=Subsurface
 }
 nodes [
  Add(A=WorldPosition() {worldPositionShaderOffset=WPT_ExcludeAllShaderOffsets}.XYZ, B=NamedRerouteUsage())
  CameraVectorWS()
  n1 = WorldPosition() {worldPositionShaderOffset=WPT_CameraRelativeNoOffsets}
  n2 = StaticSwitchParameter(True=Add(A=@n1.XYZ, B=NamedRerouteUsage()), False=@n1.XYZ) {group=Billboard, parameterName=UseFallbackBillboard}
  n3 = NamedRerouteUsage()
  n4 = NamedRerouteUsage()
  n5 = ScalarParameter() {defaultValue=1, group=Color, parameterName=LeafVisualMergeDistance}
  n6 = Multiply(A=Round(None=Divide(A=@n4, B=@n5)), B=@n5)
  n7 = TransformPosition(None=StaticSwitchParameter(True=NamedRerouteDeclaration(None=@n6) {name=LeafPivot_LocalSpace_Merged, nodeColor=<rgba 0 1 0.5547 1>}, False=NamedRerouteUsage()) {group=BillBoard WPO, parameterName=UsePivotsMerging}) {transformSourceType=TRANSFORMPOSSOURCE_Instance, transformType=TRANSFORMPOSSOURCE_TranslatedWorld}
  n8 = Normalize(VectorInput=@n7)
  n9 = DotProduct(A=Subtract(A=Add(A=@n2, B=@n3), B=@n7), B=@n8)
  n10 = Max(
   A=Subtract(A=Multiply(A=@n9) {constB=-1}, B=ScalarParameter() {defaultValue=66.7177, group=BillBoard WPO, parameterName=BillboardSqueze_Clamp})
   B=Constant()
  )
  Ceil(None=@n10)
  n11 = Transform(None=Constant3Vector() {constant=<rgba 0 0 -1 0>}) {transformSourceType=TRANSFORMSOURCE_Camera}
  n12 = Transform(None=Constant3Vector() {constant=<rgba -1 0 0 0>}) {transformSourceType=TRANSFORMSOURCE_Camera}
  n13 = NamedRerouteUsage()
  n14 = TransformPosition(None=NamedRerouteUsage()) {transformSourceType=TRANSFORMPOSSOURCE_Instance, transformType=TRANSFORMPOSSOURCE_World}
  n15 = Multiply(
   A=MaterialFunctionCall() {materialFunction=&ObjectScale}.Scale Z
   B=Constant() {r=1}
  )
  n16 = MaterialFunctionCall(
   UVs=MaterialFunctionCall(Plane Normal Axis=@n11, Plane X Axis=@n12, Plane Center=Add(A=@n13, B=@n14), UV Scale=@n15) {materialFunction=&VirtualPlaneCoordinates}.UVs
   Texture Scale=ScalarParameter() {defaultValue=702.2277, group=Billboard, parameterName=VirtualPlaneUVSclae}
  ) {materialFunction=&ScaleUVsByCenter}
  n17 = ComponentMask(None=TextureSampleParameter2D(UVs=@n16.UVs) {group=Textures, parameterName=Internal Lines Texture, texture=/WildEstate/Environment/TestContent/WildEstate_Tree_Leaf}.RGB) {r=true}
  Ceil(None=Subtract(A=@n17, B=ScalarParameter() {defaultValue=0.6928, parameterName=LineThickness}))
  n18 = ScalarParameter() {defaultValue=1.0161, group=Hedge, parameterName=Hedge_QuantizeNormalsSteps}
  n19 = Multiply(A=VertexNormalWS(), B=@n18)
  n20 = Normalize(VectorInput=Divide(A=Round(None=@n19), B=@n18))
  n21 = CollectionParameter() {
   collection=&FortniteMaterialParameters
   parameterName=SunlightDirectionalVector
  }
  n22 = Normalize(VectorInput=Multiply(A=ComponentMask(None=@n21) {b=true, g=true, r=true}, B=Constant3Vector() {constant=<rgba 1 1 0 0>}))
  n23 = DotProduct(A=@n20, B=@n22)
  Ceil(None=@n23)
  n24 = Constant3Vector() {constant=<rgba 0 0 1 0>}
  n25 = DotProduct(A=@n20, B=Normalize(VectorInput=CrossProduct(A=@n24, B=@n22)))
  Ceil(None=@n25)
  Constant3Vector() {constant=<rgba 1 0 0 0>}
  Constant3Vector() {constant=<rgba 1 1 0 0>}
  Distance()
  n26 = MaterialFunctionCall(Float3=TransformPosition(None=Constant3Vector()) {transformSourceType=TRANSFORMPOSSOURCE_Instance, transformType=TRANSFORMPOSSOURCE_World}) {materialFunction=&BreakOutFloat3Components}
  n27 = Constant() {r=500}
  n28 = Constant() {r=1100}
  n29 = Add(A=Sine(None=Divide(A=@n26.R, B=@n27)), B=Sine(None=Divide(A=@n26.G, B=@n28)))
  Divide(A=@n29)
  n30 = ScalarParameter() {defaultValue=-5.2303, parameterName=Contrast_EdgeBillboardMask_1}
  n31 = ScalarParameter() {defaultValue=3.1093, parameterName=Contrast_EdgeBillboardMask_2}
  n32 = DotProduct(A=VertexNormalWS(), B=CameraVectorWS())
  LinearInterpolate(B=@n17, Alpha=Saturate(None=LinearInterpolate(A=@n30, B=@n31, Alpha=@n32))) {constA=1}
  n33 = NamedRerouteUsage()
  n34 = Subtract(A=CameraPositionWS(), B=@n33)
  MaterialFunctionCall(Vector=@n34) {materialFunction=&SafeNormalize}
  n35 = CollectionParameter() {
   collection=&FortniteMaterialParameters
   parameterName=SunlightDirectionalVector
  }
  n36 = ShadowReplace(Default=Normalize(VectorInput=@n34), Shadow=ComponentMask(None=@n35) {b=true, g=true, r=true})
  n37 = Normalize(VectorInput=VertexNormalWS())
  n38 = DotProduct(A=@n36, B=@n37)
  n39 = StaticSwitchParameter(True=Multiply(A=@n36, B=Sign(None=@n38)), False=@n36) {defaultValue=true, group=Billboard, parameterName=BillBoardShortestPathRotation}
  n40 = CrossProduct(A=@n37, B=@n39)
  MaterialFunctionCall(Vector=@n40) {materialFunction=&SafeNormalize}
  n41 = WorldPosition()
  MaterialFunctionCall(VectorToTransform=@n41.XYZ, BasisPosition=@n33) {materialFunction=&Transform3x3Matrix}
  n42 = Constant() {r=2}
  MaterialFunctionCall(
   "Animation  Phase (0-1)"=Constant()
   Number of Rows=@n42
   Number of Columns=@n42
   Texture=TextureObject() {texture=/WildEstate/Environment/TestContent/WildEstate_Bush_Leaf}
   UVs=@n16.UVs
  ) {materialFunction=&FlipBook}
  n43 = NamedRerouteUsage()
  n44 = NamedRerouteUsage()
  n45 = StaticSwitchParameter(
   True=NamedRerouteUsage()
   False=Multiply(
    A=StaticSwitchParameter(True=@n43, False=@n44) {group=VirtualPlaneShading, parameterName=UseVirtualPlaneShading, sortPriority=1}
    B=NamedRerouteUsage()
   )
  ) {group=Hedge, parameterName=HedgeShading, sortPriority=1}
  n46 = NamedRerouteUsage()
  n47 = MaterialFunctionCall(Float3=NamedRerouteUsage()) {materialFunction=&BreakOutFloat3Components}
  n48 = Add(A=Add(A=@n47.R, B=@n47.R), B=@n47.B)
  n49 = ScalarParameter() {defaultValue=33.3, group=Outlines, parameterName=ClusterRandomOutline}
  n50 = Add(A=@n46, B=Frac(None=Divide(A=@n48, B=@n49) {constB=10}))
  n51 = LinearInterpolate(
   A=Constant() {r=0.05}
   B=Constant() {r=0.49}
   Alpha=StaticSwitchParameter(True=Frac(None=@n50), False=@n46) {group=Outlines, parameterName=OutlineEachCluster}
  )
  n52 = SetMaterialAttributes!(
   MaterialAttributes
   Base Color=StaticSwitch(True=NamedRerouteUsage(), False=@n45, Value=NamedRerouteUsage())
   Specular=Constant()
   Roughness=@n51
   Opacity=Constant()
   Subsurface Color=Constant3Vector()
   Normal=StaticSwitch(True=NamedRerouteUsage(), False=NamedRerouteUsage(), Value=NamedRerouteUsage())
   Opacity Mask=TextureSampleParameter2D() {group=Textures, parameterName=OpacityMask, texture=/WildEstate/Environment/TestContent/T_WildEstate_Leaves_CutOut}.R
  ) {attributeSetTypes [
    #1
    #2
    #3
    #4
    #5
    #6
    #7
   ]}
  n53 = NamedRerouteUsage()
  n54 = NamedRerouteUsage()
  n55 = SetMaterialAttributes!(
   MaterialAttributes=@n52
   World Position Offset=Multiply(A=StaticSwitchParameter(True=@n53, False=@n54) {group=Billboard, parameterName=UseFallbackBillboard}, B=NamedRerouteUsage())
  ) {attributeSetTypes [#8]}
  n56 = MaterialFunctionCall!(In=MaterialFunctionCall!(
    In=StaticSwitchParameter(True=@n55, False=@n52) {defaultValue=true, group=Billboard, parameterName=Enable_Billboard}
    IsImpostorBake=NamedRerouteUsage()
   ) {materialFunction=&/MF_WIldEstate_Trunk}.Result) {materialFunction=&/MF_WildEstate_MediumPlantAnimation}
  NamedRerouteDeclaration(None=@n56.LeafPivot_LS) {name=LeafPivot_LocalSpace, nodeColor=<rgba 0.8047 0 1 1>}
  n57 = Multiply(A=ScalarParameter() {defaultValue=-0.99, group=BillBoard WPO, parameterName=BillboardSqueze}, B=Multiply(A=@n8, B=@n9) {constB=-1})
  n58 = Multiply(A=@n8, B=ScalarParameter() {defaultValue=-100, group=BillBoard WPO, parameterName=BillboardOffsetCamDir})
  NamedRerouteDeclaration(None=ShadowReplace(Default=Add(A=@n57, B=@n58), Shadow=Constant3Vector())) {name=WPO_LeafFlatten, nodeColor=<rgba 0 1 0.4844 1>}
  n59 = Multiply(
   A=ScalarParameter() {defaultValue=0.104, group=Color Variation WorldSpace, parameterName=RandomHueShift}
   B=Subtract(A=NamedRerouteUsage()) {constB=0.5}
  )
  n60 = LinearInterpolate(
   A=VectorParameter() {group=Color, parameterName=Color Trunk, defaultValue=<rgba 0.1042 0.0364 0.0088 1>}.RGB
   B=VectorParameter() {group=Color, parameterName=Color, defaultValue=<rgba 0.0061 0.1354 0.0559 1>}.RGB
   Alpha=NamedRerouteUsage()
  )
  n61 = TextureSampleParameter2D() {
   group=Textures
   parameterName=Diffuse
   samplerSource=SSM_Wrap_WorldGroupSettings
   texture=/Game/Characters/Fortress_Character/Textures/T_All_White
  }
  n62 = StaticSwitchParameter(True=Multiply(A=@n60, B=@n61.RGB), False=@n60) {group=Color, parameterName=EnableDiffuseTexture}
  NamedRerouteDeclaration(None=StaticSwitchParameter(
    True=MaterialFunctionCall(Hue Shift Percentage=@n59, Texture=@n62) {materialFunction=&HueShift}.Result
    False=@n62
   ) {defaultValue=true, group=Color Variation WorldSpace, parameterName=UseRandomHueColor}) {name=ColorBase, nodeColor=<rgba 0 1 0.1562 1>}
  n63 = CollectionParameter() {
   collection=&FortniteMaterialParameters
   parameterName=SunlightDirectionalVector
  }
  n64 = ComponentMask(None=@n63) {b=true, g=true, r=true}
  n65 = CameraVectorWS()
  NamedRerouteDeclaration(None=Normalize(VectorInput=Add(A=@n64, B=@n65) {desc=Compromise for Direction light and all othe rlights})) {name=NormalsFinal, nodeColor=<rgba 0 1 0.4609 1>}
  n66 = ScalarParameter() {defaultValue=0.764, group=VirtualPlaneShading, parameterName=DirectionalShadow Min Darkness}
  n67 = WorldPosition()
  n68 = ObjectPositionWS()
  n69 = TextureSampleParameter2D(UVs=MaterialFunctionCall(
    UVs=MaterialFunctionCall!(Plane Normal Axis=@n11, Plane X Axis=@n12, WorldPosition=@n67.XYZ, Plane Center=@n68, UV Scale=@n15) {materialFunction=&/VirtualPlaneCoordinates}.UVs
    Texture Scale=ScalarParameter() {defaultValue=655.1484, group=VirtualPlaneShading, parameterName=VirtualPlane UV_Scale}
   ) {materialFunction=&ScaleUVsByCenter}.UVs) {
   group=VirtualPlaneShading
   parameterName=VirtualPlaneHightMap
   texture=/WildEstate/Environment/TestContent/T_WildEstate_Tree_Noise_B_M
   channelNames {a="", b=HeightSoftMask, g=HeightHardMask, r=""}
  }
  n70 = ScalarParameter() {defaultValue=3, group=VirtualPlaneShading, parameterName=StepsB2}
  n71 = Divide(A=Round(None=Multiply(A=@n69.HeightHardMask, B=@n70)), B=@n70)
  n72 = ScalarParameter() {defaultValue=0.2606, group=VirtualPlaneShading, parameterName=Heightmap_Strength_Hard}
  n73 = Multiply(A=@n69.HeightSoftMask, B=ScalarParameter() {defaultValue=0.2606, group=VirtualPlaneShading, parameterName=Heightmap_Strength_Soft})
  n74 = WorldPosition()
  n75 = ObjectPositionWS()
  n76 = CollectionParameter() {
   collection=&FortniteMaterialParameters
   parameterName=SunlightDirectionalVector
  }
  n77 = DotProduct(A=Normalize(VectorInput=Subtract(A=@n74.XYZ, B=@n75)), B=ComponentMask(None=@n76) {b=true, g=true, r=true})
  n78 = ScalarParameter() {defaultValue=2, group=VirtualPlaneShading, parameterName=StepsB}
  n79 = Round(None=Multiply(A=Add(A=@n73, B=@n77), B=@n78))
  n80 = Add(
   A=Add(A=Multiply(A=@n71, B=@n72), B=Divide(A=@n79, B=@n78))
   B=ScalarParameter() {defaultValue=-0.6278, group=VirtualPlaneShading, parameterName=Threshold Shadowing}
  )
  NamedRerouteDeclaration(None=Saturate(None=LinearInterpolate(A=@n66, Alpha=@n80))) {name=DotVirtualPlaneShading, nodeColor=<rgba 1 0 0.5156 1>}
  n81 = ScalarParameter() {defaultValue=0.512, group=Color, parameterName=Shading_min}
  n82 = ScalarParameter() {defaultValue=1.1151, group=Color, parameterName=Shading_Max}
  n83 = WorldPosition()
  n84 = ObjectPositionWS()
  n85 = CollectionParameter() {
   collection=&FortniteMaterialParameters
   parameterName=SunlightDirectionalVector
  }
  n86 = DotProduct(A=Normalize(VectorInput=Subtract(A=@n83.XYZ, B=@n84)), B=ComponentMask(None=@n85) {b=true, g=true, r=true})
  n87 = ScalarParameter() {defaultValue=2, group=Color, parameterName=DualShading_Range, sortPriority=99}
  n88 = LocalPosition()
  n89 = ComponentMask(None=PreSkinnedLocalBounds().Max) {b=true}
  n90 = Add(
   A=VectorParameter() {group=Color, parameterName=Shading Center LS, defaultValue=<rgba 0 0 0 1>}.RGB
   B=AppendVector(A=Constant2Vector(), B=Divide(A=@n89))
  )
  n91 = Divide(A=Length(None=Subtract(A=@n88.XYZ, B=@n90)) {is=class MaterialExpressionLength}) {constB=100}
  n92 = ScalarParameter() {defaultValue=0.1, group=Color, parameterName=DualShading_Leaf_Length_Contribution, sortPriority=99}
  n93 = NamedRerouteUsage()
  n94 = NamedRerouteUsage()
  n95 = Subtract(A=VertexInterpolator(VS=StaticSwitchParameter(True=@n93, False=@n94) {group=Color, parameterName=UseLeafClusterMergeForShading}).PS, B=@n90)
  n96 = VectorParameter() {group=Color, parameterName=Shading Scale, defaultValue=<rgba 1 1 1 1>}
  n97 = Normalize(VectorInput=Transform(None=Multiply(A=@n95, B=@n96.RGB)) {transformSourceType=TRANSFORMSOURCE_Local})
  n98 = CollectionParameter() {
   collection=&FortniteMaterialParameters
   parameterName=SunlightDirectionalVector
  }
  n99 = DotProduct(A=@n97, B=ComponentMask(None=@n98) {b=true, g=true, r=true})
  n100 = Add(
   A=Add(A=Divide(A=@n86, B=@n87), B=Multiply(A=@n91, B=@n92))
   B=Multiply(A=@n99, B=ScalarParameter() {defaultValue=0.1, group=Color, parameterName=DualShading_Leaf_Contribution, sortPriority=99})
  )
  n101 = ScalarParameter() {defaultValue=-0.5, group=Color, parameterName=DualShading_Bias, sortPriority=99}
  n102 = ScalarParameter() {defaultValue=3, group=Color, parameterName=DualShading_Steps, sortPriority=99}
  n103 = Round(None=Multiply(A=Add(A=@n100, B=@n101), B=@n102))
  n104 = ScalarParameter() {defaultValue=1, desc="1 is default and even amount of light and shade on the tree. ", group=Color, parameterName=Shading_Bias}
  n105 = ScalarParameter() {
   defaultValue=2
   desc="Shading Range on the Dot product 2 = full range. From -1 to 1 of the dot product.\r\nlower value makes the shaded area smaller \r\nhigher value makes the shaded area larger"
   group=Color
   parameterName=Shading_Range
  }
  n106 = ScalarParameter() {
   defaultValue=8
   desc="Shading Range on the Dot product 2 = full range. From -1 to 1 of the dot product.\r\nlower value makes the shaded area smaller \r\nhigher value makes the shaded area larger"
   group=Color
   parameterName=Shading_Steps
  }
  n107 = Multiply(A=Divide(A=Add(A=@n99, B=@n104), B=@n105), B=@n106)
  n108 = StaticSwitchParameter(True=Saturate(None=Divide(A=@n103, B=@n102)), False=Divide(A=Round(None=@n107), B=@n106)) {group=Color, parameterName=UseDualShading, sortPriority=98}
  NamedRerouteDeclaration(None=Saturate(None=LinearInterpolate(A=@n81, B=@n82, Alpha=@n108))) {name=MergedLeafDotShading, nodeColor=<rgba 1 0.1641 0 1>}
  NamedRerouteDeclaration(None=@n97) {name=NormalsForImpostor, nodeColor=<rgba 1 0 0.2344 1>}
  NamedRerouteDeclaration(None=@n62) {name=ColorBase_ForImpostorBake, nodeColor=<rgba 0 1 0.1562 1>}
  NamedRerouteDeclaration(None=MaterialFunctionCall(Number=VertexInterpolator(VS=@n89).PS, UVs=ScreenPosition().ViewportUV) {materialFunction=&DebugScalarValues}.Result) {name=DebugTest, nodeColor=<rgba 0.5234 0 1 1>}
  NamedRerouteDeclaration(None=@n56.LeafPivot_WS) {name=LeafPivot_WorldSpace, nodeColor=<rgba 0.8047 0 1 1>}
  NamedRerouteDeclaration() {name=Leaf XAxis_ WorldSpace, nodeColor=<rgba 0.8047 0 1 1>}
  n109 = Saturate(None=Sign(None=@n25))
  n110 = Sign(None=@n23)
  n111 = Multiply(
   A=LinearInterpolate(A=ScalarParameter() {defaultValue=0.88, group=Hedge, parameterName=HedgeShadingMidShade}, Alpha=Saturate(None=@n109))
   B=LinearInterpolate(A=ScalarParameter() {defaultValue=0.75, group=Hedge, parameterName=HedgeShadingShadow}, Alpha=Saturate(None=@n110))
  )
  n112 = ScalarParameter() {defaultValue=1.25, group=Hedge, parameterName=HedgeShadingTopShade}
  n113 = DotProduct(A=@n20, B=@n24)
  n114 = Saturate(None=Sign(None=@n113))
  NamedRerouteDeclaration(None=Multiply(A=LinearInterpolate(A=@n111, B=@n112, Alpha=@n114) {constA=1}, B=NamedRerouteUsage())) {name=HedgeShading, nodeColor=<rgba 0.4297 0 1 1>}
  NamedRerouteDeclaration(None=@n56.LeafMask) {name=Leaf Mask, nodeColor=<rgba 0 1 0.3203 1>}
  NamedRerouteDeclaration(None=@n56.WPO Only) {name=Anim_WPO_only, nodeColor=<rgba 0.8047 0 1 1>}
  NamedRerouteDeclaration(None=VertexColor().R) {name=PerLeafRandom, nodeColor=<rgba 1 0 0.6308 1>}
  NamedRerouteDeclaration(None=Divide(A=Add(A=@n29) {constB=2}) {constB=4}) {name=PerActorRandom, nodeColor=<rgba 0 1 0.7891 1>}
  n115 = NamedRerouteUsage()
  n116 = ScalarParameter() {parameterName=InteriorLines_MinOpacity}
  NamedRerouteDeclaration(None=LinearInterpolate(B=@n17, Alpha=Max(A=@n115, B=@n116)) {constA=1}) {name=Interior Lines Virtual PlaneBillboard, nodeColor=<rgba 1 0 0.3984 1>}
  NamedRerouteDeclaration(None=@n56.Anim_BendOnly) {name=Anim_WPO_OverallBend, nodeColor=<rgba 0.8047 0 1 1>}
  n117 = Subtract(A=PixelDepth().R, B=ScalarParameter() {defaultValue=400, group=Outlines, parameterName=InteriorLines_Fade_Offset})
  n118 = ScalarParameter() {defaultValue=1600, group=Outlines, parameterName=InteriorLines_Fade_Distance}
  NamedRerouteDeclaration(None=Saturate(None=Divide(A=@n117, B=@n118))) {name=NearFarMask, nodeColor=<rgba 0 1 0.2031 1>}
  n119 = DotProduct(A=@n39, B=@n37)
  n120 = MaterialFunctionCall(Multiplier=Constant() {r=0.5}) {materialFunction=&Pi}
  n121 = RotateAboutAxis(
   NormalizedRotationAxis=Normalize(VectorInput=@n40)
   RotationAngle=Multiply(A=Arccosine!(None=@n119), B=Divide(A=@n120.Result) {constB=10})
   PivotPoint=@n33
   Position=@n41.XYZ
  )
  NamedRerouteDeclaration(None=Add(A=NamedRerouteUsage(), B=NamedRerouteDeclaration(None=@n121) {name=BillboardTransformFallback, nodeColor=<rgba 1 0.1641 0 1>})) {name=WPO BillboardTransformFallback, nodeColor=<rgba 1 0.1641 0 1>}
  NamedRerouteUsage()
  NamedRerouteUsage()
  ObjectLocalBounds()
  ObjectPositionWS()
  ObjectPositionWS()
  ObjectRadius()
  Round(None=@n23)
  Round(None=@n113)
  n122 = Multiply(A=@n9) {constB=-1}
  n123 = ScalarParameter() {defaultValue=104.7052, group=BillBoard WPO, parameterName=Billboard_EdgeMask}
  Saturate(None=Divide(A=Subtract(A=@n122, B=@n123)) {constB=0.0001})
  ScalarParameter() {defaultValue=1000, group=Color Variation WorldSpace, parameterName=Color Variation Distance WS, sortPriority=10}
  ScalarParameter() {defaultValue=1100, group=Color Variation WorldSpace, parameterName=Color Variation Distance WS2, sortPriority=10}
  n124 = StaticBool() {desc=Don't change}
  n125 = Constant() {r=1}
  n126 = MaterialFunctionCall!(In=@n56.Result, Enabled, Disabled, useDirectionalShading, DirectionalShadingMask) {materialFunction=/WildEstate/Environment/BaseMaterials/MaterialFunctions/MF_Test5050Unlit}
  n127 = StaticSwitchParameter(
   True=SetMaterialAttributes(MaterialAttributes=@n126.Result, World Position Offset=Constant3Vector()) {attributeSetTypes [#8]}
   False=@n126.Result
  ) {parameterName=DisableWPO}
  n128 = GetMaterialAttributes(None=@n127) {attributeGetTypes [#1]}
  n129 = MaterialFunctionCall(
   Color_Morning=VectorParameter() {parameterName=Morning, defaultValue=<rgba 1 1 1 1>}.RGB
   Color_Day=VectorParameter() {parameterName=Day, defaultValue=<rgba 0.8021 0.8021 0.8021 1>}.RGB
   Color_Evening=VectorParameter() {parameterName=Evening, defaultValue=<rgba 1 1 1 1>}.RGB
   Color_Night=VectorParameter() {parameterName=Night, defaultValue=<rgba 0.2996 0.4717 1 1>}.RGB
  ) {materialFunction=&Color_Time_of_Day}
  n130 = StaticSwitchParameter(
   True=SetMaterialAttributes!(MaterialAttributes=@n128.MaterialAttributes, Emissive Color=Multiply(A=@n128.Base Color, B=@n129.Result)) {attributeSetTypes [#9]}
   False=@n127
  ) {desc=Don't Use This is for the S37 Only, parameterName=IsUnlitMaterial}
  n131 = StaticSwitch(
   True=SetMaterialAttributes!(
    MaterialAttributes=MaterialFunctionCall!(Material Attributes=@n130, IsTangentSpaceNormal=@n124, In=@n125) {materialFunction=/Game/ImpostorBaker/MaterialFunctions/ImposterCaptureSwitch}.Result
    World Position Offset=Constant3Vector()
   ) {attributeSetTypes [#8]}
   False=@n130
   Value=NamedRerouteDeclaration(None=StaticBool()) {name=IsImpostorBake, nodeColor=<rgba 0 1 1 1>}
  )
  n132 = NamedRerouteUsage()
  n133 = WorldPosition()
  SetMaterialAttributes(
   MaterialAttributes=@n131
   World Position Offset=Multiply(A=Subtract(A=@n132, B=@n133.XYZ), B=ScalarParameter() {defaultValue=1.1442, parameterName=Testeste})
  ) {attributeSetTypes [#8]}
  ShadowReplace(Default=@n121, Shadow=Constant3Vector())
  StaticSwitchParameter(True=Multiply(A=NamedRerouteUsage(), B=@n45), False=@n45) {defaultValue=true, parameterName=UseVirtual Plane Card}
  n134 = Multiply(A=@n8, B=Saturate(None=@n10))
  n135 = ScalarParameter() {defaultValue=-206.2373, group=BillBoard WPO, parameterName=BillboardSqueze_PushInner}
  StaticSwitchParameter(True=@n57, False=Add(A=@n57, B=Multiply(A=@n134, B=@n135))) {defaultValue=true, parameterName=UseVirtual Plane Card}
  n136 = GetMaterialAttributes(None=@n131) {attributeGetTypes [#8]}
  n137 = Constant3Vector()
  n138 = Subtract(A=PixelDepth().R, B=ScalarParameter() {defaultValue=7500, group=WPO Fade, parameterName=WPO_Fade_Offset})
  n139 = ScalarParameter() {defaultValue=1500, group=WPO Fade, parameterName=WPO_Fade_Distance}
  n140 = Saturate(None=NaniteReplace!(
    Default=Divide(A=@n138, B=@n139)
    Nanite=MaterialFunctionCall!(FadeBlendDistance=@n139, AutoHandleBadDistance) {materialFunction=/Game/Global/MaterialFunctions/TechArt/MF_WPODisableBlend}.VertexShaderOutput
   ))
  StaticSwitchParameter(
   True=SetMaterialAttributes(MaterialAttributes=@n136.MaterialAttributes, World Position Offset=LinearInterpolate(A=@n136.World Position Offset, B=@n137, Alpha=@n140)) {attributeSetTypes [#8]}
   False=@n131
  ) {defaultValue=true, group=WPO Fade, parameterName=UseWPOFade}
  StaticSwitchParameter(True=Constant() {r=0.678}, False=@n51) {group=Hedge, parameterName=HedgeShading, sortPriority=1}
  StaticSwitchParameter(True=TwoSidedSign(), False=@n125) {parameterName=UseFlipBackFaceNormals}
  TransformPosition() {transformSourceType=TRANSFORMPOSSOURCE_Instance, transformType=TRANSFORMPOSSOURCE_TranslatedWorld}
  Transform(None=@n3) {transformSourceType=TRANSFORMSOURCE_World, transformType=TRANSFORM_Camera}
  WorldPosition()
 ]
}
