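# Two-sided subsurface foliage material for WildEstate pine leaves that paints stylized, stepped shadow bands onto a tint color, supporting Nanite and instancing.
# Floored height-gradient and sun-direction thresholds, broken up by noise on camera-facing per-instance plane UVs, multiply into Base Color.
# The world-space normal blends sun and camera vectors as a lighting compromise.
# Snow tint and flipbook leaf-texture branches are left unconnected.
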Material /WildEstate/Foliage/BaseMaterial/M_WildEstate_Pine_Leaf
graph {
 properties {
  bAutomaticallySetUsageInEditor=false
  bTangentSpaceNormal=false
  bUseMaterialAttributes=true
  bUsedWithInstancedStaticMeshes=true
  bUsedWithNanite=true
  shadingModel=MSM_Subsurface
  twoSided=true
  usedShadingModels=Subsurface
 }
 nodes [
  Constant3Vector()
  n1 = Constant3Vector() {constant=<rgba 0 0 1 0>}
  n2 = VertexNormalWS()
  Divide(A=Add(A=DotProduct(A=@n1, B=@n2)))
  n3 = VectorParameter() {parameterName=Color, defaultValue=<rgba 0.0031 0.1979 0.0832 1>}
  LinearInterpolate(
   A=@n3.RGB
   B=VectorParameter() {parameterName=SnowColor, defaultValue=<rgba 0.5273 0.8617 1.25 1>}.RGB
   Alpha=ComponentMask(None=VertexNormalWS()) {b=true}
  )
  LocalPosition()
  Multiply()
  n4 = Constant3Vector()
  n5 = MaterialFunctionCall(Float3=Divide(A=TransformPosition(None=@n4) {transformSourceType=TRANSFORMPOSSOURCE_Instance, transformType=TRANSFORMPOSSOURCE_World}) {constB=10}) {materialFunction=&BreakOutFloat3Components}
  n6 = Frac(None=Add(A=Add(A=@n5.R, B=@n5.R), B=@n5.B))
  n7 = Constant() {r=2}
  n8 = TextureObject() {texture=/WildEstate/Environment/TestContent/WildEstate_Bush_Leaf}
  n9 = MaterialFunctionCall() {materialFunction=&ObjectScale}
  n10 = MaterialFunctionCall(
   Plane Normal Axis=CameraVectorWS()
   Plane X Axis=Transform(None=Constant3Vector() {constant=<rgba -1 0 0 0>}) {transformSourceType=TRANSFORMSOURCE_Camera}
   Plane Center=TransformPosition(None=Constant3Vector()) {transformSourceType=TRANSFORMPOSSOURCE_Instance, transformType=TRANSFORMPOSSOURCE_World}
   UV Scale=Multiply(A=@n9.Scale Z, B=Constant() {r=1})
  ) {materialFunction=&VirtualPlaneCoordinates}
  n11 = ScalarParameter() {defaultValue=296.9218, parameterName=VirtualPlaneUVSclae}
  n12 = Saturate(None=MaterialFunctionCall(UVs=@n10.UVs, Texture Scale=@n11) {materialFunction=&ScaleUVsByCenter}.UVs)
  n13 = ScalarParameter() {defaultValue=0.8517, parameterName=BottomShadowMin2}
  n14 = Add(A=ComponentMask(None=LocalPosition().Z) {b=true}, B=ScalarParameter() {defaultValue=-58.9089, parameterName=BottomShadowGradientOffset})
  n15 = ScalarParameter() {defaultValue=1044.1434, parameterName=BottomShadowGradientDist}
  n16 = Divide(A=@n10.UVs, B=ScalarParameter() {defaultValue=1516.2898, parameterName=BushNoise_Scale})
  n17 = TextureSample(UVs=ComponentMask(None=@n16) {r=true}) {texture=/WildEstate/Environment/TestContent/T_WildEstatePine_Noise_M}
  n18 = ScalarParameter() {defaultValue=-0.1481, parameterName=NoiseAdd_LighBottomGradient}
  n19 = Saturate(None=Add(A=Divide(A=@n14, B=@n15), B=Multiply(A=@n17.R, B=@n18)))
  n20 = Add(A=@n19, B=ScalarParameter() {defaultValue=0.5372, parameterName=BottomShadowThreshold2})
  n21 = Saturate(None=Floor(None=Saturate(None=@n20)))
  n22 = ScalarParameter() {defaultValue=0.7, parameterName=BottomShadowMin}
  n23 = Add(A=@n19, B=ScalarParameter() {defaultValue=0.8652, parameterName=BottomShadowThreshold})
  n24 = Saturate(None=Floor(None=Saturate(None=@n23)))
  n25 = Multiply(A=Saturate(None=Add(A=@n13, B=@n21)), B=Saturate(None=Add(A=@n22, B=@n24)))
  n26 = CollectionParameter() {
   collection=&FortniteMaterialParameters
   parameterName=SunlightDirectionalVector
  }
  n27 = ComponentMask(None=@n26) {b=true, g=true, r=true}
  n28 = VertexNormalWS()
  n29 = Normalize(VectorInput=Multiply(A=@n28, B=Constant3Vector() {constant=<rgba 1 1 0 0>}))
  n30 = Multiply(
   A=TextureSample(UVs=@n16) {texture=&T_WildEstatePine_Noise_M}.R
   B=ScalarParameter() {defaultValue=0.5239, parameterName=NoiseAdd_LightDirectionGradient}
  )
  n31 = DotProduct(A=@n27, B=Normalize(VectorInput=Add(A=@n29, B=@n30)))
  n32 = Add(A=@n31, B=ScalarParameter() {defaultValue=0.481, parameterName=Threshold})
  n33 = TwoSidedSign()
  n34 = Floor(None=LinearInterpolate(A=Constant(), B=Saturate(None=@n32), Alpha=Saturate(None=@n33)))
  n35 = Saturate(None=Add(A=ScalarParameter() {defaultValue=0.868, parameterName=DirectionShadowMin}, B=Saturate(None=@n34)))
  n36 = ScalarParameter() {defaultValue=0.9907, parameterName=Threshold2}
  n37 = LinearInterpolate(A=Constant(), B=Saturate(None=Add(A=@n31, B=@n36)), Alpha=Saturate(None=TwoSidedSign()))
  n38 = Add(A=ScalarParameter() {defaultValue=0.684, parameterName=DirectionShadowMin2}, B=Saturate(None=Floor(None=@n37)))
  n39 = Multiply(A=Multiply(A=Multiply(A=@n25, B=@n35), B=Saturate(None=@n38)), B=@n3.RGB)
  Multiply(
   A=ComponentMask(None=MaterialFunctionCall("Animation  Phase (0-1)"=@n6, Number of Rows=@n7, Number of Columns=@n7, Texture=@n8, UVs=@n12) {materialFunction=&FlipBook}.Result) {r=true}
   B=@n39
  )
  n40 = TwoSidedSign()
  Multiply(A=@n28, B=AppendVector(A=AppendVector(A=@n40, B=@n40), B=Constant() {r=1}))
  Multiply(A=@n9.Scale Z, B=@n11)
  Normalize(VectorInput=Subtract(A=WorldPosition().XYZ, B=CameraPositionWS()))
  ObjectPositionWS()
  ObjectPositionWS()
  ScalarParameter() {defaultValue=0.868, parameterName=DirectionShadowMin}
  n41 = Transform(None=@n27) {transformSourceType=TRANSFORMSOURCE_World}
  n42 = CameraVectorWS()
  SetMaterialAttributes!(
   MaterialAttributes
   Base Color=@n39
   Metallic
   Specular=Constant()
   Roughness
   Emissive Color
   Opacity=ScalarParameter() {parameterName=Opacity}
   Subsurface Color=Constant3Vector()
   World Position Offset
   Normal=Normalize(VectorInput=Add(A=@n41, B=@n42) {desc=Compromise for Direction light and all othe rlights})
  ) {attributeSetTypes [
    #1
    #2
    #3
    #4
    #5
    #6
    #7
    #8
    #9
   ]}
  ShadowReplace(Default=Constant3Vector())
  Subtract(A=WorldPosition().XYZ, B=ObjectPositionWS())
  Subtract(
   A=WorldPosition().XYZ
   B=Add(A=ActorPositionWS(), B=VectorParameter() {desc=Temp, parameterName=CanopyCenter, defaultValue=<rgba 0 0 859.1398 1>}.RGB)
  )
  TextureSample(UVs=@n12) {texture=&WildEstate_Bush_Leaf}
 ]
}
