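# Subsurface-shaded, Nanite-enabled base material for WildEstate tree leaves, built around one tintable BaseColor with no textures.
# Sun direction from FortniteMaterialParameters dotted with the vertex normal, offset by Threshold, then saturated and floored, gives a hard toon-like mask.
# That mask switches BaseColor to an Emissive-boosted version.
# Parameters scale diffuse and subsurface color and expose roughness, specular and opacity; output wiring is absent.
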Material /WildEstate/Foliage/BaseMaterial/M_WildEstate_TreeLeaves
graph {
 properties {bAutomaticallySetUsageInEditor=false, bUsedWithNanite=true, shadingModel=MSM_Subsurface, usedShadingModels=Subsurface}
 nodes [
  DotProduct()
  n1 = VectorParameter() {parameterName=BaseColor, defaultValue=<rgba 0.0314 0.151 0.0593 1>}
  EyeAdaptationInverse(LightValueInput=@n1.RGB)
  n2 = CollectionParameter() {
   collection=&FortniteMaterialParameters
   parameterName=SunlightDirectionalVector
  }
  n3 = Add(
   A=DotProduct(A=ComponentMask(None=@n2) {b=true, g=true, r=true}, B=VertexNormalWS())
   B=ScalarParameter() {defaultValue=1.2207, parameterName=Threshold}
  )
  LinearInterpolate(A=@n1.RGB, B=Multiply(A=@n1.RGB, B=ScalarParameter() {defaultValue=2.9915, parameterName=Emissive}), Alpha=Floor(None=Saturate(None=@n3)))
  Multiply(A=@n1.RGB, B=ScalarParameter() {parameterName=SSSBrightness})
  Multiply(A=@n1.RGB, B=ScalarParameter() {parameterName=DiffuseBrightness})
  Multiply(A=EyeAdaptation().EyeAdaptation, B=@n1.RGB)
  ScalarParameter() {defaultValue=1, parameterName=Roughness}
  ScalarParameter() {parameterName=Specular}
  ScalarParameter() {parameterName=SSSOpacity}
 ]
}
