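# Material function that makes big foliage bushes shrink and thin as the camera nears, through two static switches on incoming material attributes.
# Leaf shrink offsets vertices toward the input pivot and multiplies opacity mask by a texture biased by camera distance, blending two fades around a bush core.
# Sphere shrink then overrides world position offset, collapsing vertices onto SphereCenter by instance-to-camera distance.
# Screen-size and core-shrink branches are left unconnected.
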MaterialFunction /WildEstate/Foliage/BaseMaterial/MF_WildEstateBigBush
graph {nodes [
  n1 = VectorParameter() {parameterName=SphereCenter, defaultValue=<rgba 0 0 0 1>}
  n2 = Constant3Vector()
  n3 = Subtract(
   A=Length(None=TransformPosition(None=@n2) {transformSourceType=TRANSFORMPOSSOURCE_Instance, transformType=TRANSFORMPOSSOURCE_TranslatedWorld}) {is=class MaterialExpressionLength}
   B=ScalarParameter() {defaultValue=1600, group=BigBush, parameterName=BigBush_Sphere_Fade_Offset}
  )
  n4 = ScalarParameter() {defaultValue=200, group=BigBush, parameterName=BigBush_Sphere_Fade_Blend}
  n5 = LinearInterpolate(
   A=Subtract(
    A=TransformPosition(None=@n1.RGB) {transformSourceType=TRANSFORMPOSSOURCE_Instance, transformType=TRANSFORMPOSSOURCE_TranslatedWorld}
    B=WorldPosition() {worldPositionShaderOffset=WPT_CameraRelative}.XYZ
   )
   Alpha=Saturate(None=Divide(A=@n3, B=@n4))
  )
  n6 = FunctionInput() {inputType=FunctionInput_MaterialAttributes}
  n7 = GetMaterialAttributes(None=@n6) {attributeGetTypes [#1, #2]}
  n8 = NamedRerouteDeclaration(None=@n7.WorldPositionOffset) {name=WPOCurrent, nodeColor=<rgba 0.875 0 1 1>}
  n9 = FunctionInput() {inputName=Pivot WorldSPace, sortPriority=10}
  n10 = NamedRerouteUsage()
  n11 = WorldPosition()
  n12 = Distance(A=CameraPositionWS(), B=@n9)
  n13 = ScalarParameter() {defaultValue=100, group=BigBush, parameterName=BigBush_LeafShrink_Fade_Offset}
  n14 = Saturate(None=Divide(A=Subtract(A=@n12, B=@n13), B=ScalarParameter() {defaultValue=325.7803, group=BigBush, parameterName=BigBush_LeafShrink_Fade_Blend}))
  n15 = LinearInterpolate(A=Subtract(A=@n9, B=Add(A=@n10, B=@n11.XYZ)), Alpha=@n14) {constB=0}
  n16 = Multiply(A=TextureCoordinate(), B=ScalarParameter() {defaultValue=1.5, group=BigBush, parameterName=BigBush_Mask_Scale})
  n17 = ScalarParameter() {group=BigBush, parameterName=BigBush_Mask_NearValue}
  n18 = ScalarParameter() {defaultValue=1, group=BigBush, parameterName=BigBush_Mask_FarValue}
  n19 = Divide(
   A=Subtract(A=@n12, B=ScalarParameter() {defaultValue=399.2122, group=BigBush, parameterName=BigBush_Mask_Fade_Offset})
   B=ScalarParameter() {defaultValue=600, group=BigBush, parameterName=BigBush_Mask_Fade_Blend}
  )
  n20 = Divide(
   A=Subtract(A=@n12, B=ScalarParameter() {defaultValue=200, group=BigBush, parameterName=BigBush_Mask_Fade_Offset2})
   B=ScalarParameter() {defaultValue=600, group=BigBush, parameterName=BigBush_Mask_Fade_Blend2}
  )
  n21 = Length(None=TransformPosition(None=VectorParameter() {parameterName=CenterBushMask, defaultValue=<rgba 0 0 0 1>}.RGB) {transformType=TRANSFORMPOSSOURCE_TranslatedWorld})
  n22 = Divide(
   A=Subtract(A=@n21, B=ScalarParameter() {defaultValue=500, group=BigBush, parameterName=BigBush_Mask_FadeCore_Offset})
   B=ScalarParameter() {defaultValue=50, group=BigBush, parameterName=BigBush_Mask_FadeCore_Blend}
  )
  n23 = Saturate(None=LinearInterpolate(A=Saturate(None=@n19), B=Saturate(None=@n20), Alpha=Saturate(None=@n22)))
  n24 = Saturate(None=Add(
    A=TextureSampleParameter2D(UVs=@n16) {parameterName=BigBushMask, texture=/WildEstate/Foliage/LargeBush/Textures/T_WildEstate_Bush_MaskDF_M}.R
    B=LinearInterpolate(A=@n17, B=@n18, Alpha=@n23)
   ))
  n25 = StaticSwitchParameter(
   True=SetMaterialAttributes!(MaterialAttributes=@n7.MaterialAttributes, WorldPositionOffset=Add(A=@n8, B=@n15), OpacityMask=Multiply(A=@n7.OpacityMask, B=@n24)) {attributeSetTypes [#1, #2]}
   False=@n6
  ) {group=BigBush, parameterName=IsBigBush_NearLeafShrink}
  FunctionOutput(None=StaticSwitchParameter(
    True=SetMaterialAttributes!(MaterialAttributes=@n25, WorldPositionOffset=@n5) {attributeSetTypes [#1]}
    False=@n25
   ) {group=BigBush, parameterName=IsBigBush_NearSphereShrink})
  Multiply(
   A=ScalarParameter() {defaultValue=0.0005, group=BigBush, parameterName=BigBush_Mask_ScreenSize}
   B=Divide(A=ScreenPosition().ViewportUV, B=SceneTexelSize())
  )
  n26 = Divide(
   A=Subtract(A=@n21, B=ScalarParameter() {group=BigBush, parameterName=BigBush_LeafShrinkCore_Fade_Offset})
   B=ScalarParameter() {defaultValue=100, group=BigBush, parameterName=BigBush_LeafShrinkCore_Fade_Blend}
  )
  Multiply(A=@n14, B=Max(A=ScalarParameter() {defaultValue=0.75, group=BigBush, parameterName=BigBush_LeafShrinkCore_Min}, B=Saturate(None=@n26)))
  Saturate(None=LinearInterpolate(Alpha=@n24))
 ]}
