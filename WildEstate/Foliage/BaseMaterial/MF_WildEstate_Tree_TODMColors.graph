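# Material function for Wild Estate trees that tints their color by time of day and renders it as emissive.
# It multiplies the incoming BaseColor by the Color_Time_of_Day result, fed by locked-group morning, day, evening and night color parameters, and writes it to EmissiveColor.
# MaterialProxyReplace turns BaseColor black at runtime, while keeping the original color for material proxy baking.
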MaterialFunction /WildEstate/Foliage/BaseMaterial/MF_WildEstate_Tree_TODMColors
graph {nodes [
  n1 = GetMaterialAttributes(None=FunctionInput() {inputName=Result, inputType=FunctionInput_MaterialAttributes}) {attributeGetTypes [#1]}
  n2 = MaterialFunctionCall(
   Color_Morning=VectorParameter() {group=DONTOUCH, parameterName=DONTTOUCHMorning, defaultValue=<rgba 0.7292 0.5854 0.5216 1>}.RGB
   Color_Day=VectorParameter() {group=DONTOUCH, parameterName=DONTTOUCHDay, defaultValue=<rgba 0.8021 0.8021 0.8021 1>}.RGB
   Color_Evening=VectorParameter() {group=DONTOUCH, parameterName=DONTTOUCHEvening, defaultValue=<rgba 0.7604 0.6947 0.4924 1>}.RGB
   Color_Night=VectorParameter() {group=DONTOUCH, parameterName=DONTTOUCHNight, defaultValue=<rgba 0.2996 0.4717 1 1>}.RGB
  ) {materialFunction=&Color_Time_of_Day}
  n3 = Constant3Vector()
  FunctionOutput(None=SetMaterialAttributes!(
    MaterialAttributes=@n1.MaterialAttributes
    EmissiveColor=Multiply(A=@n1.BaseColor, B=@n2.Result)
    BaseColor=MaterialProxyReplace(Realtime=@n3, MaterialProxy=@n1.BaseColor)
   ) {attributeSetTypes [#2, #1]}) {outputName=SetMaterialAttributes}
 ]}
