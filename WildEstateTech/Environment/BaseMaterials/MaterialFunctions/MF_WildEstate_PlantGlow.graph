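# Material function that adds a pulsing emissive glow to plant materials, enabled by the BushHerb static switch.
# A height gradient from instance bounds and a vertex-normal Z mask shape the tinted glow, with optional Fresnel.
# A per-instance-offset sine pulses it, time-of-day values scale it and inverse eye adaptation compensates exposure.
# A parameter blends between additive and lerped emissive; a bounding-box pulse branch is left unconnected.
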MaterialFunction /WildEstateTech/Environment/BaseMaterials/MaterialFunctions/MF_WildEstate_PlantGlow
graph {nodes [
  n1 = FunctionInput() {inputType=FunctionInput_MaterialAttributes}
  n2 = GetMaterialAttributes(None=@n1) {attributeGetTypes [#1]}
  n3 = NamedRerouteUsage()
  n4 = LinearInterpolate(
   A=Add(A=@n2.EmissiveColor, B=@n3)
   B=LinearInterpolate(A=@n2.EmissiveColor, B=@n3, Alpha=NamedRerouteUsage())
   Alpha=ScalarParameter() {group=BushHerb, parameterName=glow_UseLerpOrAdd}
  )
  FunctionOutput(None=StaticSwitchParameter(
    True=SetMaterialAttributes!(MaterialAttributes=@n2.MaterialAttributes, EmissiveColor=@n4) {attributeSetTypes [#1]}
    False=@n1
   ) {group=BushHerb, parameterName=BushHerb})
  LocalPosition()
  MaterialFunctionCall() {materialFunction=/Game/Effects/Niagara/MasterMaterials/Functions/InstanceBounds_0-1_UVW}
  n5 = Power(
   Base=ComponentMask(None=MaterialFunctionCall() {materialFunction=&BoundingBoxBased_0-1_UVW}.B) {b=true}
   Exp=ScalarParameter() {defaultValue=2, group=BushHerb, parameterName=glow_Pow}
  )
  n6 = LinearInterpolate(
   A=ScalarParameter() {group=BushHerb, parameterName=glow_offset_Z_low}
   B=ScalarParameter() {group=BushHerb, parameterName=glow_offset_Z}
   Alpha=MaterialFunctionCall(Speed=ScalarParameter() {defaultValue=1, group=BushHerb, parameterName=Pulse_Speed}) {materialFunction=&DebugTimeSine}.Result
  )
  n7 = Multiply(A=OneMinus(None=Floor(None=@n5)), B=OneMinus(None=Add(A=@n5, B=@n6)))
  Multiply(A=NamedRerouteUsage(), B=Saturate(None=@n7))
  n8 = Add(A=MaterialFunctionCall(UV Channel=TextureCoordinate()) {materialFunction=&LinearGradient}.VGradient) {constB=-0.5}
  Multiply(A=@n7, B=Saturate(None=Multiply(A=@n8) {constB=256}))
  NamedRerouteDeclaration(None=VectorParameter() {group=BushHerb, parameterName=NucularGlowColor, defaultValue=<rgba 0.0752 1 0.1728 1>}.RGB) {name=Glow_Color, nodeColor=<rgba 1 0.3047 0 1>}
  n9 = ScalarParameter() {defaultValue=0.9311, group=BushHerb, parameterName=Light_Offset}
  n10 = LinearInterpolate(B=ScalarParameter() {defaultValue=0.008, group=BushHerb, parameterName=Light_Random_Add}, Alpha=Saturate(None=PerInstanceRandom())) {constB=0.5}
  n11 = Power(
   Base=Add(
    A=MaterialFunctionCall() {materialFunction=&InstanceBounds_0-1_UVW}.B
    B=Add(A=@n9, B=@n10) {constB=0.1}
   )
   Exp=ScalarParameter() {defaultValue=3.9572, group=BushHerb, parameterName=Light_Pow}
  )
  n12 = VertexNormalWS()
  n13 = Multiply(
   A=Add(A=ComponentMask(None=@n12) {b=true}, B=ScalarParameter() {group=BushHerb, parameterName=vertex_offset})
   B=ScalarParameter() {defaultValue=1.5, group=BushHerb, parameterName=vertex_Str}
  )
  n14 = Multiply(A=Saturate(None=OneMinus(None=@n11)), B=Saturate(None=OneMinus(None=@n13)))
  NamedRerouteDeclaration(None=@n14) {name=GradientMask, nodeColor=<rgba 0.7109 1 0 1>}
  n15 = Fresnel(ExponentIn=ScalarParameter() {defaultValue=5, group=BushHerb, parameterName=Fresnel_exp0}, Normal=VertexNormalWS())
  n16 = Time()
  n17 = Multiply(A=PerInstanceRandom()) {constB=2}
  n18 = Sine(None=Multiply(A=Add(A=@n16, B=@n17), B=ScalarParameter() {defaultValue=0.2, parameterName=speed}))
  n19 = Multiply(
   A=Multiply(A=NamedRerouteUsage(), B=@n14)
   B=Multiply(A=ScalarParameter() {defaultValue=5, group=BushHerb, parameterName=Light_Str}, B=ConstantBiasScale(None=@n18) {bias=2})
  )
  n20 = StaticSwitchParameter(True=Multiply(A=Power(Base=@n15), B=@n19), False=@n19) {group=BushHerb, parameterName=Fresnel_Glow?}
  n21 = MaterialFunctionCall(
   Value_Morning=ScalarParameter() {defaultValue=1, group=BushHerb, parameterName="Glow_Value_Morning (S)"}
   Value_Day=ScalarParameter() {defaultValue=0.95, group=BushHerb, parameterName="Glow_Value_Day (S)"}
   Value_Evening=ScalarParameter() {defaultValue=0.95, group=BushHerb, parameterName="Glow_Value_Evening (S)"}
   Value_Night=ScalarParameter() {defaultValue=0.8, group=BushHerb, parameterName="Glow_Value_Night (S)"}
  ) {materialFunction=&Scalar_Time_of_Day}
  NamedRerouteDeclaration(None=EyeAdaptationInverse(LightValueInput=Multiply(A=@n20, B=@n21.Result)).EyeAdaptationInverse) {name=light_Color, nodeColor=<rgba 1 0.4453 0 1>}
  PerInstanceRandom()
  VectorParameter() {group=NucularGlow, parameterName=Light_Color, defaultValue=<rgba 0.0752 1 0.1728 1>}
 ]}
