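# Base subsurface, Nanite-ready foliage material for stylized WildEstate trees, hedges and herbs, with trunk, glow-card and impostor-bake switches.
# Color combines per-instance hue and value variation with posterized sun-direction shading from leaf pivots, virtual planes or hedge normals, overriding normals toward sun plus camera.
# World position offset combines animation functions and leaf billboarding, then fades by distance.
# Destructible damage, plant glow and effect functions follow, with per-platform branches.
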Material /WildEstate/Foliage/BaseMaterial/M_WildEstate_Tree_BaseMaterial
graph {
 properties {
  bAutomaticallySetUsageInEditor=false
  bHasPixelAnimation=true
  bTangentSpaceNormal=false
  bUseMaterialAttributes=true
  bUsedWithInstancedStaticMeshes=true
  bUsedWithNanite=true
  shadingModel=MSM_Subsurface
  usedShadingModels=Subsurface
 }
 nodes [
  Add(A=WorldPosition() {worldPositionShaderOffset=WPT_ExcludeAllShaderOffsets}.XYZ, B=NamedRerouteUsage())
  n1 = NamedRerouteUsage()
  n2 = CollectionParameter() {
   collection=&FortniteMaterialParameters
   parameterName=SunlightDirectionalVector
  }
  n3 = WorldPosition() {worldPositionShaderOffset=WPT_CameraRelative}
  n4 = Constant2Vector()
  n5 = Divide(A=ComponentMask(None=Bounds().Max) {b=true})
  n6 = TransformPosition(None=Add(A=VectorParameter() {parameterName=Center, defaultValue=<rgba 0 0 0 1>}.RGB, B=AppendVector(A=@n4, B=@n5))) {transformSourceType=TRANSFORMPOSSOURCE_Instance, transformType=TRANSFORMPOSSOURCE_TranslatedWorld}
  n7 = NaniteReplace!(
   Default=Multiply(A=VectorParameter() {group=Hedge, parameterName=HedgeShadingRoundScale, defaultValue=<rgba 1 1 1 1>}.RGB, B=Subtract(A=@n3.XYZ, B=@n6))
   Nanite=VertexNormalWS()
  )
  n8 = DotProduct(A=ComponentMask(None=@n2) {b=true, g=true, r=true}, B=@n7)
  Add(
   A=Multiply(
    A=ComponentMask(None=@n1) {r=true}
    B=ScalarParameter() {defaultValue=-548425.5625, group=Hedge, parameterName=HedgShadingRound_VirtTexture_Add}
   )
   B=@n8
  )
  Bounds()
  CameraVectorWS()
  n9 = WorldPosition() {worldPositionShaderOffset=WPT_CameraRelativeNoOffsets}
  n10 = StaticSwitchParameter(True=Add(A=@n9.XYZ, B=NamedRerouteUsage()), False=@n9.XYZ) {group=Billboard, parameterName=UseFallbackBillboard}
  n11 = NamedRerouteUsage()
  n12 = ScalarParameter() {defaultValue=150, group=Color, parameterName=LeafVisualMergeDistance}
  n13 = ScalarParameter() {defaultValue=150, group=Color, parameterName=LeafVisualMergeDistanceLow}
  n14 = MaterialFunctionCall("Xb1,PS4"=@n12, "PC-High&&Epic,XSX,PS5"=@n12, "PC-Perf,Android"=@n13, Switch=@n13, "PC-Low(NotPerf)&&Medium"=@n12) {materialFunction=&PlatformSwitch_1D}
  n15 = Round(None=Divide(A=NamedRerouteUsage(), B=@n14.Result))
  n16 = TransformPosition(None=NamedRerouteDeclaration(None=Multiply(A=@n15, B=@n14.Result)) {name=LeafPivot_LocalSpace_Merged, nodeColor=<rgba 0 1 0.5547 1>}) {transformSourceType=TRANSFORMPOSSOURCE_Instance, transformType=TRANSFORMPOSSOURCE_TranslatedWorld}
  n17 = Normalize(VectorInput=@n16)
  n18 = DotProduct(A=Subtract(A=Add(A=@n10, B=@n11), B=@n16), B=@n17)
  n19 = Max(
   A=Subtract(A=Multiply(A=@n18) {constB=-1}, B=ScalarParameter() {defaultValue=66.7177, group=BillBoard WPO, parameterName=BillboardSqueze_Clamp})
   B=Constant()
  )
  Ceil(None=@n19)
  n20 = Transform(None=Constant3Vector() {constant=<rgba 0 0 -1 0>}) {transformSourceType=TRANSFORMSOURCE_Camera}
  n21 = Transform(None=Constant3Vector() {constant=<rgba -1 0 0 0>}) {transformSourceType=TRANSFORMSOURCE_Camera}
  n22 = NamedRerouteUsage()
  n23 = TransformPosition(None=NamedRerouteUsage()) {transformSourceType=TRANSFORMPOSSOURCE_Instance, transformType=TRANSFORMPOSSOURCE_World}
  n24 = Multiply(
   A=MaterialFunctionCall() {materialFunction=&ObjectScale}.Scale Z
   B=Constant() {r=1}
  )
  n25 = MaterialFunctionCall(
   UVs=MaterialFunctionCall(Plane Normal Axis=@n20, Plane X Axis=@n21, Plane Center=Add(A=@n22, B=@n23), UV Scale=@n24) {materialFunction=&VirtualPlaneCoordinates}.UVs
   Texture Scale=ScalarParameter() {defaultValue=702.2277, group=Billboard, parameterName=VirtualPlaneUVSclae}
  ) {materialFunction=&ScaleUVsByCenter}
  n26 = ComponentMask(None=TextureSampleParameter2D(UVs=@n25.UVs) {group=Textures, parameterName=Internal Lines Texture, texture=/WildEstate/Environment/TestContent/WildEstate_Tree_Leaf}.RGB) {r=true}
  Ceil(None=Subtract(A=@n26, B=ScalarParameter() {defaultValue=0.6928, parameterName=LineThickness}))
  n27 = ScalarParameter() {defaultValue=-9.5449, group=Hedge, parameterName=Hedge_mainAxisEdgeContrast_low}
  n28 = ScalarParameter() {defaultValue=50, group=Hedge, parameterName=Hedge_mainAxisEdgeContrast_high}
  n29 = VectorParameter() {group=Hedge, parameterName=HedgeMainTilingAxis, defaultValue=<rgba 0 1 0 1>}
  n30 = Abs(None=DotProduct(A=VertexNormalWS(), B=Transform(None=@n29.RGB) {transformSourceType=TRANSFORMSOURCE_Instance}))
  n31 = ObjectPositionWS() {originType=CameraRelative}
  n32 = DotProduct(
   A=VectorParameter() {group=Hedge, parameterName=HedgeNormalModDirectionY, defaultValue=<rgba 0 1 0 1>}.RGB
   B=Subtract(A=LocalPosition().XYZ, B=TransformPosition(None=@n31) {transformSourceType=TRANSFORMPOSSOURCE_TranslatedWorld})
  )
  n33 = Multiply(
   A=Saturate(None=LinearInterpolate(A=@n27, B=@n28, Alpha=@n30))
   B=Multiply(A=VectorParameter() {group=Hedge, parameterName=HedgeNormalMod, defaultValue=<rgba -0.3816 0 0 1>}.RGB, B=Sign(None=@n32))
  )
  n34 = ScalarParameter() {defaultValue=0.8315, group=Hedge, parameterName=Hedge_QuantizeNormalsSteps}
  n35 = Multiply(A=Add(A=VertexNormalWS(), B=Transform(None=@n33) {transformSourceType=TRANSFORMSOURCE_Instance}), B=@n34)
  n36 = Normalize(VectorInput=Divide(A=Round(None=@n35), B=@n34))
  n37 = CollectionParameter() {
   collection=&FortniteMaterialParameters
   parameterName=SunlightDirectionalVector
  }
  n38 = Normalize(VectorInput=Multiply(A=ComponentMask(None=@n37) {b=true, g=true, r=true}, B=Constant3Vector() {constant=<rgba 1 1 0 0>}))
  n39 = DotProduct(A=@n36, B=@n38)
  Ceil(None=@n39)
  n40 = Constant3Vector() {constant=<rgba 0 0 1 0>}
  n41 = DotProduct(A=@n36, B=Normalize(VectorInput=CrossProduct(A=@n40, B=@n38)))
  Ceil(None=@n41)
  n42 = Custom!(None) {
   code="//float3 Center = Parameters.SceneData.InstanceData.LocalBoundsCenter;\r\n//float3 Center = Parameters.SceneData.Primitive.MaxWPOExtent;\r\n\r\nfloat3 Center = Parameters.SceneData.InstanceData.LocalBoundsCenter;\r\nreturn Center;"
   desc=Work around for Pre-Skinned LocalBouns not working with ISM
   showCode=true
  }
  ComponentMask(None=VertexInterpolator(VS=@n42).PS) {b=true}
  Constant3Vector() {constant=<rgba 1 0 0 0>}
  Constant3Vector() {constant=<rgba 1 1 0 0>}
  Constant3Vector() {constant=<rgba 0 0 1 0>}
  Custom() {
   code="MaterialFloat3 CustomExpression0(FMaterialVertexParameters Parameters)\r\nreturn Parameters.SceneData.Primitive.InstanceWPODisableDistanceSquared;"
   showCode=true
  }
  Distance()
  n43 = MaterialFunctionCall(Float3=TransformPosition(None=Constant3Vector()) {transformSourceType=TRANSFORMPOSSOURCE_Instance, transformType=TRANSFORMPOSSOURCE_World}) {materialFunction=&BreakOutFloat3Components}
  n44 = Constant() {r=500}
  n45 = Constant() {r=1100}
  n46 = Add(A=Sine(None=Divide(A=@n43.R, B=@n44)), B=Sine(None=Divide(A=@n43.G, B=@n45)))
  Divide(A=@n46)
  Divide(A=Add(A=@n46) {constB=2}) {constB=4}
  n47 = ScalarParameter() {defaultValue=-5.2303, parameterName=Contrast_EdgeBillboardMask_1}
  n48 = ScalarParameter() {defaultValue=3.1093, parameterName=Contrast_EdgeBillboardMask_2}
  n49 = DotProduct(A=VertexNormalWS(), B=CameraVectorWS())
  LinearInterpolate(B=@n26, Alpha=Saturate(None=LinearInterpolate(A=@n47, B=@n48, Alpha=@n49))) {constA=1}
  n50 = GetMaterialAttributes(None=NamedRerouteUsage()) {attributeGetTypes [#1]}
  n51 = MaterialFunctionCall!(
   Center Location=WorldPosition().XYZ
   0-1 UVs
   XY scale=ScalarParameter() {defaultValue=35, group=Emissive, parameterName=GlowCardScale}
   OverridePivot
   "Normals (Optional)"
  ) {materialFunction=&Sprite}
  n52 = StaticBool() {desc=Don't change}
  n53 = Constant() {r=1}
  n54 = MaterialFunctionCall(
   UVs=TextureCoordinate()
   CenterPosition=Constant2Vector() {g=0.5, r=0.5}
   Radius=Constant() {r=0.5}
   Density=ScalarParameter() {defaultValue=1.0313, group=Emissive, parameterName=EmissiveGlowFalloff}
  ) {materialFunction=&RadialGradientExponential}
  n55 = ScalarParameter() {defaultValue=1.0313, group=Emissive, parameterName=EmissiveGlowDepthFade}
  n56 = StaticSwitchParameter(True=NamedRerouteUsage(), False=NamedRerouteUsage()) {group=VirtualPlaneShading, parameterName=UseVirtualPlaneShading, sortPriority=1}
  n57 = NamedRerouteUsage()
  n58 = ScalarParameter() {defaultValue=0.8, group=Color, parameterName=BackFaceBrightness}
  n59 = Saturate(None=TwoSidedSign())
  n60 = StaticSwitchParameter(True=Multiply(A=@n57, B=LinearInterpolate(A=@n58, Alpha=@n59)), False=@n57) {group=Color, parameterName=UseTwoSidedBackFaceDarkening}
  n61 = StaticSwitchParameter(
   True=NamedRerouteUsage()
   False=StaticSwitchParameter(True=NamedRerouteUsage(), False=Multiply(A=@n56, B=@n60)) {group=Hedge, parameterName=HedgeShading, sortPriority=1}
  ) {group=Herb, parameterName=HerbShading, sortPriority=1}
  n62 = Constant() {r=0.05}
  n63 = Constant() {r=0.49}
  n64 = NamedRerouteUsage()
  n65 = MaterialFunctionCall(Float3=NamedRerouteUsage()) {materialFunction=&BreakOutFloat3Components}
  n66 = Add(A=@n65.R, B=@n65.R)
  n67 = Frac(None=Divide(A=Add(A=@n66, B=@n65.B), B=ScalarParameter() {defaultValue=33.3, group=Outlines, parameterName=ClusterRandomOutline}) {constB=10})
  n68 = StaticSwitchParameter(True=Frac(None=Add(A=@n64, B=@n67)), False=@n64) {group=Outlines, parameterName=OutlineEachCluster}
  n69 = SetMaterialAttributes!(
   MaterialAttributes
   Base Color=StaticSwitch(True=NamedRerouteUsage(), False=@n61, Value=NamedRerouteUsage())
   Specular=Constant()
   Roughness=StaticSwitchParameter(
    True=ScalarParameter() {bUseCustomPrimitiveData=true, defaultValue=0.678, parameterName=CustomRoughneRoughness, primitiveDataIndex=26}
    False=LinearInterpolate(A=@n62, B=@n63, Alpha=@n68)
   ) {group=Hedge, parameterName=HedgeShading, sortPriority=1}
   Opacity=Constant()
   Subsurface Color=Constant3Vector()
   Normal=StaticSwitch(True=NamedRerouteUsage(), False=NamedRerouteUsage(), Value=NamedRerouteUsage())
   Opacity Mask=TextureSampleParameter2D() {group=Textures, parameterName=OpacityMask, texture=/WildEstate/Environment/TestContent/T_WildEstate_Leaves_CutOut}.R
  ) {attributeSetTypes [
    #2
    #3
    #4
    #5
    #6
    #7
    #8
   ]}
  n70 = GetMaterialAttributes(None=@n69) {attributeGetTypes [#4]}
  n71 = TextureSampleParameter2D() {
   group=Textures
   parameterName=Custom Roughness Texture
   texture=/WildEstate/Foliage/MediumPlants/BushHerbPlant/Textures/T_BushHerbRed_Consumable_Rough
  }
  n72 = LinearInterpolate(A=@n70.Roughness, B=@n71.RGB, Alpha=ScalarParameter() {defaultValue=0.5, group=Textures, parameterName=Custom Roughness Blend})
  n73 = PixelDepth()
  n74 = ScalarParameter() {defaultValue=150, group=Textures, parameterName=CustomRoughness_Fade_Offset}
  n75 = Clamp(None=Divide(A=Subtract(A=@n73.R, B=@n74), B=ScalarParameter() {defaultValue=1500, group=Textures, parameterName=CustomRoughness_Fade_Distance}))
  n76 = StaticSwitchParameter(
   True=BlendMaterialAttributes(
    A=SetMaterialAttributes!(MaterialAttributes=@n70.MaterialAttributes, Roughness=@n72) {attributeSetTypes [#4]}
    B=@n70.MaterialAttributes
    Alpha=Saturate(None=@n75)
   )
   False=@n69
  ) {group=Textures, parameterName=Use Custom Roughness}
  n77 = SetMaterialAttributes!(
   MaterialAttributes=@n76
   World Position Offset=StaticSwitchParameter(True=NamedRerouteUsage(), False=NamedRerouteUsage()) {group=Billboard, parameterName=UseFallbackBillboard}
  ) {attributeSetTypes [#1]}
  n78 = StaticSwitchParameter(
   True=SetMaterialAttributes!(
    MaterialAttributes
    Emissive Color=VectorParameter() {group=Emissive, parameterName=Emissive_Glow_Color, defaultValue=<rgba 1 1 1 1>}.RGB
    Opacity=DepthFade(Opacity=@n54.RadialGradientExponential, FadeDistance=@n55)
    Opacity Mask=@n54.RadialGradientExponential
   ) {attributeSetTypes [#9, #5, #8]}
   False=MaterialFunctionCall!(
    In=StaticSwitchParameter(True=@n77, False=@n76) {defaultValue=true, group=Billboard, parameterName=Enable_Billboard}
    IsImpostorBake=NamedRerouteUsage()
   ) {materialFunction=&/MF_WIldEstate_Trunk}.Result
  ) {group=Emissive, parameterName=IsTranslucentGlowCard}
  n79 = MaterialFunctionCall!(In=@n78, OpaqueCanopyMask) {materialFunction=&/MF_TreeAnim_Apollo_WildEstate}
  n80 = MaterialFunctionCall!(In=@n78) {materialFunction=&/MF_WildEstate_Tree_Animation_2}
  n81 = StaticSwitchParameter(True=@n79.Result, False=@n80.Result) {parameterName=UseFallBackAnimation}
  n82 = MaterialFunctionCall!(In=@n81, Pivot WorldSPace=NamedRerouteUsage()) {materialFunction=&/MF_WildEstateBigBush}
  n83 = MaterialFunctionCall!(In=@n82.Result, Enabled, Disabled, useDirectionalShading, DirectionalShadingMask) {materialFunction=/WildEstate/Environment/BaseMaterials/MaterialFunctions/MF_Test5050Unlit}
  n84 = StaticSwitchParameter(
   True=@n83.Result
   False=StaticSwitch(
    True=@n82.Result
    False=MaterialFunctionCall!(Result=@n82.Result) {materialFunction=&/MF_WildEstate_Tree_TODMColors}.SetMaterialAttributes
    Value=NamedRerouteUsage()
   )
  ) {group=Trunk, parameterName=IsTrunk}
  n85 = StaticSwitchParameter(
   True=SetMaterialAttributes!(MaterialAttributes=@n84, World Position Offset=Constant3Vector()) {attributeSetTypes [#1]}
   False=@n84
  ) {parameterName=DisableWPO}
  n86 = StaticSwitch(
   True=SetMaterialAttributes(
    MaterialAttributes=MaterialFunctionCall!(Material Attributes=@n85, IsTangentSpaceNormal=@n52, In=@n53) {materialFunction=/Game/ImpostorBaker/MaterialFunctions/ImposterCaptureSwitch}.Result
    World Position Offset=Constant3Vector()
   ) {attributeSetTypes [#1]}
   False=@n85
   Value=NamedRerouteDeclaration(None=StaticBool()) {name=IsImpostorBake, nodeColor=<rgba 0 1 1 1>}
  )
  n87 = GetMaterialAttributes(None=@n86) {attributeGetTypes [#1]}
  n88 = ScalarParameter() {defaultValue=1500, group=WPO Fade, parameterName=WPO_Fade_Distance}
  n89 = Divide(A=Subtract(A=PixelDepth().R, B=ScalarParameter() {defaultValue=7500, group=WPO Fade, parameterName=WPO_Fade_Offset}), B=@n88)
  n90 = MaterialFunctionCall!(FadeBlendDistance=@n88, AutoHandleBadDistance) {materialFunction=/Game/Global/MaterialFunctions/TechArt/MF_WPODisableBlend}
  n91 = LinearInterpolate(A=@n87.World Position Offset, B=Constant3Vector(), Alpha=Saturate(None=NaniteReplace(Default=@n89, Nanite=@n90.VertexShaderOutput)))
  n92 = StaticSwitchParameter(
   True=SetMaterialAttributes(MaterialAttributes=@n50.MaterialAttributes, World Position Offset=Add(A=@n50.World Position Offset, B=@n51.World position offset)) {attributeSetTypes [#1]}
   False=StaticSwitchParameter(
    True=SetMaterialAttributes(MaterialAttributes=@n87.MaterialAttributes, World Position Offset=@n91) {attributeSetTypes [#1]}
    False=@n86
   ) {defaultValue=true, group=WPO Fade, parameterName=UseWPOFade}
  ) {group=Emissive, parameterName=IsTranslucentGlowCard}
  n93 = MaterialFunctionCall!(
   Material_In=@n92
   UseDestructibleDamage
   --- usage info tooltip ---
   UseCrackedDamageTextures=StaticBool()
   UseDamageWobbleWPO
   UseAdditiveWPO=StaticBool() {value=true}
   UseMobileWPO
   WorldPosition
   UseDestructionMultiplier
   ExposeLerpFunctionForInvertingTheHitDistanceCalc
   InvertDamageFalloff
   UseAlternateHitGradientOutput
   DestructibleDamageAmount
   WSDeformationCenter
   WSDamageRadius
  ) {materialFunction=&/MF_FN_DestructibleDamage_WildEstate}
  n94 = BlendMaterialAttributes(
   A=SetMaterialAttributes(
    MaterialAttributes=@n93.Material_Out
    World Position Offset=GetMaterialAttributes(None=@n92) {attributeGetTypes [#1]}.World Position Offset
   ) {attributeSetTypes [#1]}
   B=@n93.Material_Out
   Alpha=NaniteReplace(Default=ScalarParameter() {defaultValue=1, group=Optimziation, parameterName=NonNanite_OverrideWPO, sliderMax=1}, Nanite=Constant() {r=1})
  )
  n95 = TextureSampleParameter2D() {
   group=Wild Estate Tech Art FX
   parameterName=BushHerb_SFXMask
   samplerType=SAMPLERTYPE_Masks
   texture=/WildEstate/Foliage/MediumPlants/BushHerbPlant/Textures/T_BushHerb_Consumable_SFX
  }
  n96 = MaterialFunctionCall!(
   In=MaterialFunctionCall!(In=MaterialFunctionCall!("Xb1,PS4"=@n94, "PC-High&&Epic,XSX,PS5"=@n94, "PC-Perf,Android"=@n93.Material_Out, Switch=@n94, "PC-Low(NotPerf)&&Medium"=@n94) {materialFunction=/Game/Effects/Niagara/MasterMaterials/Functions/Common/PlatformSwitch_MA}.Result) {materialFunction=/WildEstateTech/Environment/BaseMaterials/MaterialFunctions/MF_WildEstate_PlantGlow}.Result
   Mask=@n95.R
  ) {materialFunction=/WildEstateTech/Environment/BaseMaterials/MaterialFunctions/MF_WildEstate_BushHerb_ZP}
  MaterialFunctionCall!(In=@n96.Result, UseSphericalMappingOverride) {materialFunction=/Game/Packages/Fortress_SharedMaterials/Base_Material_Functions/MF_CurieFallbackEffect}
  n97 = NamedRerouteUsage()
  n98 = Subtract(A=CameraPositionWS(), B=@n97)
  MaterialFunctionCall(Vector=@n98) {materialFunction=&SafeNormalize}
  n99 = CollectionParameter() {
   collection=&FortniteMaterialParameters
   parameterName=SunlightDirectionalVector
  }
  n100 = ShadowReplace(Default=Normalize(VectorInput=@n98), Shadow=ComponentMask(None=@n99) {b=true, g=true, r=true})
  n101 = Normalize(VectorInput=VertexNormalWS())
  n102 = DotProduct(A=@n100, B=@n101)
  n103 = StaticSwitchParameter(True=Multiply(A=@n100, B=Sign(None=@n102)), False=@n100) {defaultValue=true, group=Billboard, parameterName=BillBoardShortestPathRotation}
  n104 = CrossProduct(A=@n101, B=@n103)
  MaterialFunctionCall(Vector=@n104) {materialFunction=&SafeNormalize}
  n105 = WorldPosition()
  MaterialFunctionCall(VectorToTransform=@n105.XYZ, BasisPosition=@n97) {materialFunction=&Transform3x3Matrix}
  n106 = SetMaterialAttributes!(
   MaterialAttributes=GetMaterialAttributes(None=@n96.Result) {attributeGetTypes [#7]}.MaterialAttributes
   Normal=Constant3Vector() {constant=<rgba 0 0 1 0>}
  ) {attributeSetTypes [#7]}
  MaterialFunctionCall!(In=MaterialFunctionCall("Xb1,PS4"=@n96.Result, "PC-High&&Epic,XSX,PS5"=@n96.Result, "PC-Perf,Android"=@n106, Switch=@n106, "PC-Low(NotPerf)&&Medium"=@n106) {materialFunction=&PlatformSwitch_MA}.Result) {
   desc="never set during actual season, remove if art direction calls out foliage being slightly brighter"
   materialFunction=/Game/Effects/Niagara/MasterMaterials/Functions/MF_FX_MAEmissiveAdjustments
  }
  MaterialFunctionCall(UseCrackedDamageTextures=StaticBool(), UseAdditiveWPO=StaticBool() {value=true}) {materialFunction=&MF_FN_DestructibleDamage}
  n107 = Constant() {r=2}
  MaterialFunctionCall(
   "Animation  Phase (0-1)"=Constant()
   Number of Rows=@n107
   Number of Columns=@n107
   Texture=TextureObject() {texture=/WildEstate/Environment/TestContent/WildEstate_Bush_Leaf}
   UVs=@n25.UVs
  ) {materialFunction=&FlipBook}
  Multiply(A=Subtract(A=@n79.Leaf_Pivot_WS, B=WorldPosition().XYZ), B=ScalarParameter() {defaultValue=0.896, parameterName=Testeste})
  Multiply(
   A=Normalize(VectorInput=NamedRerouteUsage())
   B=VectorParameter() {group=Herb, parameterName=ShadingScale_Herbs, defaultValue=<rgba 12 12 1 1>}.RGB
  )
  NamedRerouteDeclaration(None=StaticSwitchParameter(True=@n79.Leaf_Pivot_LS, False=@n80.LeafPivot_LocalSpace) {parameterName=UseFallBackAnimation}) {name=LeafPivot_LocalSpace, nodeColor=<rgba 0.8047 0 1 1>}
  n108 = ScalarParameter() {defaultValue=-0.99, group=BillBoard WPO, parameterName=BillboardSqueze}
  n109 = ScalarParameter() {defaultValue=-0.99, group=BillBoard WPO, parameterName=BillboardSquezeLowEnd}
  n110 = Multiply(
   A=MaterialFunctionCall("Xb1,PS4"=@n108, "PC-High&&Epic,XSX,PS5"=@n108, "PC-Perf,Android"=@n109, Switch=@n109, "PC-Low(NotPerf)&&Medium"=@n108) {materialFunction=&PlatformSwitch_1D}.Result
   B=Multiply(A=@n17, B=@n18) {constB=-1}
  )
  n111 = Multiply(A=@n17, B=ScalarParameter() {defaultValue=-100, group=BillBoard WPO, parameterName=BillboardOffsetCamDir})
  NamedRerouteDeclaration(None=ShadowReplace(Default=Add(A=@n110, B=@n111), Shadow=Constant3Vector())) {name=WPO_LeafFlatten, nodeColor=<rgba 0 1 0.4844 1>}
  n112 = ScalarParameter() {bUseCustomPrimitiveData=true, defaultValue=0.65, group=Color, parameterName=ValueVariation_primdata, primitiveDataIndex=5}
  n113 = PerInstanceCustomData!(DefaultValue) {constDefaultValue=0.65, dataIndex=5}
  n114 = NamedRerouteUsage()
  n115 = Add(A=Multiply(A=NamedRerouteUsage()) {constB=1.33}) {constB=0.33}
  n116 = Multiply(
   A=LinearInterpolate(
    A=ScalarParameter() {defaultValue=0.65, parameterName=ValueVariation}
    B=LinearInterpolate(A=@n112, B=@n113, Alpha=@n114)
    Alpha=NamedRerouteUsage()
   )
   B=Subtract(A=Frac(None=@n115), B=ScalarParameter() {defaultValue=0.5, parameterName=ValueVariationBias}) {constB=0.5}
  )
  n117 = NamedRerouteDeclaration(None=StaticSwitch(True=Constant() {r=1}, False=Add(A=@n116), Value=NamedRerouteUsage()) {desc=For ImpostorBakeing}) {name=ValueVariation, nodeColor=<rgba 0 0.0625 1 1>}
  n118 = ScalarParameter() {defaultValue=0.104, group=Color Variation WorldSpace, parameterName=RandomHueShift}
  n119 = ScalarParameter() {bUseCustomPrimitiveData=true, defaultValue=-0.03, group=Color, parameterName=Hue Variation_primdata, primitiveDataIndex=4}
  n120 = LinearInterpolate(A=@n119, B=PerInstanceCustomData() {constDefaultValue=-0.03, dataIndex=4}, Alpha=NamedRerouteUsage())
  n121 = NamedRerouteUsage()
  n122 = NamedRerouteUsage()
  n123 = PerInstanceCustomData3Vector!(DefaultValue)
  n124 = VectorParameter() {bUseCustomPrimitiveData=true, parameterName=Color1_Base_primdata}
  n125 = PerInstanceCustomData() {dataIndex=3}
  n126 = Add(A=@n124.A, B=@n125)
  n127 = StaticSwitch(
   True=Constant3Vector() {constant=<rgba 1 1 1 0>}
   False=LinearInterpolate(
    A=VectorParameter() {group=Color, parameterName=Color, defaultValue=<rgba 0.0061 0.1354 0.0559 1>}.RGB
    B=Add(A=@n123, B=@n124.RGB)
    Alpha=NamedRerouteDeclaration(None=@n126) {name=PrimDataBlendAlpha, nodeColor=<rgba 0 1 0.3203 1>}
   )
   Value=NamedRerouteUsage()
  ) {desc=For ImpostorBakeing}
  n128 = TextureObjectParameter() {
   group=Textures
   parameterName=RandomDiffuseOutlineTexture
   samplerType=SAMPLERTYPE_Masks
   texture=/Game/Characters/Player/Female/Medium/Bodies/F_MED_Commando_01/Textures/SolidWhite_MaskCompression
  }
  n129 = NamedRerouteUsage()
  n130 = Add(A=@n129) {constB=0.33}
  n131 = StaticSwitchParameter(
   True=MaterialFunctionCall!(In=@n128, RandomPerleaf=@n129, RandomPerleaf_2=Frac(None=@n130)) {materialFunction=&/MF_WildEstate_RandomOutlineTexture}.Lerp
   False=TextureSample(Tex=TextureObjectParameter() {group=Textures, parameterName=Diffuse, texture=/Game/Characters/Fortress_Character/Textures/T_All_White}) {texture=&T_All_White}.RGB
  ) {group=Outlines, parameterName=UseRandomOutlineDiffuseTexture}
  n132 = LinearInterpolate(B=@n131, Alpha=NamedRerouteUsage()) {constA=1}
  n133 = StaticSwitchParameter(
   True=Multiply(A=@n127, B=StaticSwitchParameter(True=@n132, False=@n131) {defaultValue=true, group=Outlines, parameterName=UseOutline NearFade})
   False=@n127
  ) {group=Color, parameterName=EnableDiffuseTexture}
  n134 = MaterialFunctionCall(Hue Shift Percentage=Multiply(A=LinearInterpolate(A=@n118, B=@n120, Alpha=@n121), B=Subtract(A=@n122) {constB=0.5}), Texture=@n133) {materialFunction=&HueShift}
  NamedRerouteDeclaration(None=StaticSwitchParameter(True=Multiply(A=@n117, B=@n134.Result), False=@n133) {defaultValue=true, group=Color Variation WorldSpace, parameterName=UseRandomHueColor}) {name=ColorBase, nodeColor=<rgba 0 1 0.1562 1>}
  n135 = CollectionParameter() {
   collection=&FortniteMaterialParameters
   parameterName=SunlightDirectionalVector
  }
  n136 = Add(A=ComponentMask(None=@n135) {b=true, g=true, r=true}, B=CameraVectorWS()) {desc=Compromise for Direction light and all othe rlights}
  n137 = Constant3Vector() {constant=<rgba 0 0 1 0>}
  n138 = TwoSidedSign()
  NamedRerouteDeclaration(None=MaterialProxyReplace(Realtime=Normalize(VectorInput=@n136), MaterialProxy=Multiply(A=@n137, B=@n138))) {name=NormalsFinal, nodeColor=<rgba 0 1 0.4609 1>}
  n139 = ScalarParameter() {defaultValue=0.764, group=VirtualPlaneShading, parameterName=DirectionalShadow Min Darkness}
  n140 = ScalarParameter() {defaultValue=1, group=VirtualPlaneShading, parameterName=DirectionalShadow Max Brightness}
  n141 = WorldPosition()
  n142 = ObjectPositionWS()
  n143 = TextureSampleParameter2D(UVs=MaterialFunctionCall(
    UVs=MaterialFunctionCall!(Plane Normal Axis=@n20, Plane X Axis=@n21, WorldPosition=@n141.XYZ, Plane Center=@n142, UV Scale=@n24) {materialFunction=&/VirtualPlaneCoordinates}.UVs
    Texture Scale=ScalarParameter() {defaultValue=655.1484, group=VirtualPlaneShading, parameterName=VirtualPlane UV_Scale}
   ) {materialFunction=&ScaleUVsByCenter}.UVs) {
   group=VirtualPlaneShading
   parameterName=VirtualPlaneHightMap
   texture=/WildEstate/Environment/TestContent/T_WildEstate_Tree_Noise_B_M
   channelNames {a="", b=HeightSoftMask, g=HeightHardMask, r=""}
  }
  n144 = ScalarParameter() {defaultValue=3, group=VirtualPlaneShading, parameterName=StepsB2}
  n145 = Multiply(A=@n143.HeightHardMask, B=@n144)
  n146 = Multiply(
   A=Divide(A=Round(None=@n145), B=@n144)
   B=ScalarParameter() {defaultValue=0.2606, group=VirtualPlaneShading, parameterName=Heightmap_Strength_Hard}
  )
  n147 = Normalize(VectorInput=Subtract(A=WorldPosition().XYZ, B=NamedRerouteUsage()))
  n148 = VectorParameter() {group=VirtualPlaneShading, parameterName=Shading Scale VirtPlane, defaultValue=<rgba 1 1 1 1>}
  n149 = CollectionParameter() {
   collection=&FortniteMaterialParameters
   parameterName=SunlightDirectionalVector
  }
  n150 = Add(
   A=Multiply(A=@n143.HeightSoftMask, B=ScalarParameter() {defaultValue=0.2606, group=VirtualPlaneShading, parameterName=Heightmap_Strength_Soft})
   B=DotProduct(A=Multiply(A=@n147, B=@n148.RGB), B=ComponentMask(None=@n149) {b=true, g=true, r=true})
  )
  n151 = ScalarParameter() {defaultValue=2, group=VirtualPlaneShading, parameterName=StepsB}
  n152 = Divide(A=Round(None=Multiply(A=@n150, B=@n151)), B=@n151)
  n153 = Saturate(None=Add(A=Add(A=@n146, B=@n152), B=ScalarParameter() {group=VirtualPlaneShading, parameterName=Threshold Shadowing}))
  NamedRerouteDeclaration(None=Saturate(None=LinearInterpolate(A=@n139, B=@n140, Alpha=@n153))) {name=DotVirtualPlaneShading, nodeColor=<rgba 1 0 0.5156 1>}
  n154 = ScalarParameter() {defaultValue=0.512, group=Color, parameterName=Shading_min}
  n155 = ScalarParameter() {defaultValue=1.1151, group=Color, parameterName=Shading_Max}
  n156 = NamedRerouteUsage()
  n157 = VectorParameter() {group=Color, parameterName=Shading Scale, defaultValue=<rgba 1 1 1 1>}
  n158 = CollectionParameter() {
   collection=&FortniteMaterialParameters
   parameterName=SunlightDirectionalVector
  }
  n159 = DotProduct(A=Multiply(A=Normalize(VectorInput=@n156), B=@n157.RGB), B=ComponentMask(None=@n158) {b=true, g=true, r=true})
  n160 = ScalarParameter() {defaultValue=2, group=Color, parameterName=DualShading_Range, sortPriority=99}
  n161 = Divide(A=Length(None=@n156) {is=class MaterialExpressionLength}) {constB=100}
  n162 = ScalarParameter() {defaultValue=0.1, group=Color, parameterName=DualShading_Leaf_Length_Contribution, sortPriority=99}
  n163 = VertexInterpolator(VS=StaticSwitchParameter(True=NamedRerouteUsage(), False=NamedRerouteUsage()) {group=Color, parameterName=UseLeafClusterMergeForShading})
  n164 = ComponentMask(None=PreSkinnedLocalBounds() {desc="current not working with Instances. Might needs to be removed\r\nand Shading Center tweaked for each tree"}.Max) {b=true}
  n165 = Add(
   A=VectorParameter() {group=Color, parameterName=Shading Center LS, defaultValue=<rgba 0 0 0 1>}.RGB
   B=AppendVector(A=Constant2Vector(), B=Divide(A=@n164))
  )
  n166 = Normalize(VectorInput=Transform(None=Subtract(A=@n163.PS, B=@n165)) {transformSourceType=TRANSFORMSOURCE_Instance})
  n167 = CollectionParameter() {
   collection=&FortniteMaterialParameters
   parameterName=SunlightDirectionalVector
  }
  n168 = DotProduct(A=Multiply(A=@n166, B=@n157.RGB), B=ComponentMask(None=@n167) {b=true, g=true, r=true})
  n169 = Add(
   A=Add(A=Divide(A=@n159, B=@n160), B=Multiply(A=@n161, B=@n162))
   B=Multiply(A=@n168, B=ScalarParameter() {defaultValue=0.1, group=Color, parameterName=DualShading_Leaf_Contribution, sortPriority=99})
  )
  n170 = ScalarParameter() {defaultValue=-0.5, group=Color, parameterName=DualShading_Bias, sortPriority=99}
  n171 = ScalarParameter() {defaultValue=3, group=Color, parameterName=DualShading_Steps, sortPriority=99}
  n172 = Round(None=Multiply(A=Add(A=@n169, B=@n170), B=@n171))
  n173 = ScalarParameter() {defaultValue=1, desc="1 is default and even amount of light and shade on the tree. ", group=Color, parameterName=Shading_Bias}
  n174 = ScalarParameter() {
   defaultValue=2
   desc="Shading Range on the Dot product 2 = full range. From -1 to 1 of the dot product.\r\nlower value makes the shaded area smaller \r\nhigher value makes the shaded area larger"
   group=Color
   parameterName=Shading_Range
  }
  n175 = ScalarParameter() {
   defaultValue=8
   desc="Shading Range on the Dot product 2 = full range. From -1 to 1 of the dot product.\r\nlower value makes the shaded area smaller \r\nhigher value makes the shaded area larger"
   group=Color
   parameterName=Shading_Steps
  }
  n176 = Multiply(A=Divide(A=Add(A=@n168, B=@n173), B=@n174), B=@n175)
  n177 = StaticSwitchParameter(True=Saturate(None=Divide(A=@n172, B=@n171)), False=Divide(A=Round(None=@n176), B=@n175)) {group=Color, parameterName=UseDualShading, sortPriority=98}
  NamedRerouteDeclaration(None=Saturate(None=LinearInterpolate(A=@n154, B=@n155, Alpha=@n177))) {name=MergedLeafDotShading, nodeColor=<rgba 1 0.1641 0 1>}
  NamedRerouteDeclaration(None=@n166) {name=NormalsForImpostor, nodeColor=<rgba 1 0 0.2344 1>}
  NamedRerouteDeclaration(None=@n133) {name=ColorBase_ForImpostorBake, nodeColor=<rgba 0 1 0.1562 1>}
  NamedRerouteDeclaration(None=MaterialFunctionCall(Number=VertexInterpolator(VS=@n164).PS, UVs=ScreenPosition().ViewportUV) {materialFunction=&DebugScalarValues}.Result) {name=DebugTest, nodeColor=<rgba 0.5234 0 1 1>}
  NamedRerouteDeclaration(None=StaticSwitchParameter(
    True=@n79.Leaf_Pivot_WS
    False=TransformPosition(None=@n80.LeafPivot_LocalSpace) {transformSourceType=TRANSFORMPOSSOURCE_Instance, transformType=TRANSFORMPOSSOURCE_World}
   ) {parameterName=UseFallBackAnimation}) {name=LeafPivot_WorldSpace, nodeColor=<rgba 0.8047 0 1 1>}
  NamedRerouteDeclaration(None=@n79.Leaf_XAxis_WS) {name=Leaf XAxis_ WorldSpace, nodeColor=<rgba 0.8047 0 1 1>}
  n178 = ScalarParameter() {defaultValue=0.75, group=Hedge, parameterName=HedgeShadingShadow}
  n179 = ScalarParameter() {defaultValue=0.88, group=Hedge, parameterName=HedgeShadingMidShade}
  n180 = Subtract(A=@n8, B=ScalarParameter() {defaultValue=-1.6575, group=Hedge, parameterName=HedgShadingRound_Offset})
  n181 = LinearInterpolate(A=@n179, Alpha=Saturate(None=Ceil(None=@n180))) {constA=1}
  n182 = ScalarParameter() {defaultValue=-0.5179, group=Hedge, parameterName=HedgShadingRound_Offset2}
  n183 = Saturate(None=Ceil(None=Subtract(A=@n8, B=@n182)))
  n184 = ScalarParameter() {defaultValue=1.25, group=Hedge, parameterName=HedgeShadingTopShade}
  n185 = Round(None=Subtract(A=ComponentMask(None=@n7) {b=true}, B=ScalarParameter() {defaultValue=0.4048, group=Hedge, parameterName=HedgShadingRound_Offset_top}))
  n186 = Sign(None=@n41)
  n187 = LinearInterpolate(A=@n179, Alpha=Saturate(None=Saturate(None=@n186)))
  n188 = LinearInterpolate(A=@n178, Alpha=Saturate(None=Sign(None=@n39)))
  n189 = DotProduct(A=@n36, B=@n40)
  n190 = Sign(None=@n189)
  n191 = StaticSwitchParameter(
   True=LinearInterpolate(A=LinearInterpolate(A=@n178, B=@n181, Alpha=@n183) {constA=1}, B=@n184, Alpha=Saturate(None=@n185)) {constA=1}
   False=LinearInterpolate(A=Multiply(A=@n187, B=@n188), B=@n184, Alpha=Saturate(None=@n190)) {constA=1}
  ) {group=Hedge, parameterName=HedgeRoundShading}
  n192 = NamedRerouteUsage()
  n193 = StaticSwitchParameter(True=Multiply(A=NamedRerouteUsage(), B=@n192), False=@n192) {group=Hedge, parameterName=UseHedgeTilingTexture}
  NamedRerouteDeclaration(None=MaterialProxyReplace(Realtime=Multiply(A=@n191, B=@n193), MaterialProxy=@n192)) {name=HedgeShading, nodeColor=<rgba 0.4297 0 1 1>}
  n194 = VertexInterpolator(VS=LocalPosition().XYZ)
  NamedRerouteDeclaration(None=Transform(None=Subtract(A=@n194.PS, B=@n165)) {transformSourceType=TRANSFORMSOURCE_Instance}) {name=Position WS, nodeColor=<rgba 0 1 1 1>}
  NamedRerouteDeclaration(None=TransformPosition(None=@n165) {transformSourceType=TRANSFORMPOSSOURCE_Instance, transformType=TRANSFORMPOSSOURCE_World}) {name=CenterPosition WS, nodeColor=<rgba 0 1 1 1>}
  n195 = Subtract(A=PixelDepth().R, B=ScalarParameter() {defaultValue=500, parameterName=OutlineNearFade_Offset})
  n196 = ScalarParameter() {defaultValue=500, parameterName=OutlineNearFade_Blend}
  NamedRerouteDeclaration(None=Saturate(None=Divide(A=@n195, B=@n196))) {name=NearOutlineFade, nodeColor=<rgba 0 0.2969 1 1>}
  NamedRerouteDeclaration(None=@n125) {name=IsPerIstanceData, nodeColor=<rgba 1 0.2109 0 1>}
  n197 = Add(A=@n43.R, B=@n43.R)
  n198 = Frac(None=Divide(A=Add(A=@n197, B=@n43.B)) {constB=33.3})
  NamedRerouteDeclaration(None=@n198) {name=RandomValueforCellShading, nodeColor=<rgba 0 1 0.3672 1>}
  NamedRerouteDeclaration(None=MaterialFunctionCall(
    TextureObject=TextureObject() {texture=/WildEstate/Foliage/Hedges/Textures/T_WIldEstate_Hedge_Tiling}
    TextureSize=ScalarParameter() {defaultValue=256, group=Hedge, parameterName=HedgTilingTextureSize}
   ) {materialFunction=&WorldAlignedTexture}.XYZ Texture) {name=HedgeTilingTexture, nodeColor=<rgba 0.4297 0 1 1>}
  NamedRerouteDeclaration(None=StaticSwitchParameter(True=@n79.Anim WPO_Only, False=@n80.Anim_WPO_Only) {parameterName=UseFallBackAnimation}) {name=Anim_WPO_only, nodeColor=<rgba 0.8047 0 1 1>}
  n199 = ScalarParameter() {defaultValue=0.496, group=Herb, parameterName=HerbShading_Min}
  n200 = ScalarParameter() {defaultValue=0.8552, group=Herb, parameterName=HerbShading_Max}
  n201 = ScalarParameter() {defaultValue=2.5, group=Herb, parameterName=HerbShading_Steps}
  n202 = ScalarParameter() {defaultValue=0.184, group=Herb, parameterName=HerbShading_Bias}
  n203 = CollectionParameter() {
   collection=&FortniteMaterialParameters
   parameterName=SunlightDirectionalVector
  }
  n204 = DotProduct(A=ComponentMask(None=@n203) {b=true, g=true, r=true}, B=VertexNormalWS())
  n205 = Multiply(A=@n201, B=Multiply(A=ScalarParameter() {defaultValue=0.7948, group=Herb, parameterName=HerbShading_Range}, B=Add(A=@n202, B=@n204)))
  n206 = Saturate(None=Divide(A=Round(None=@n205), B=@n201))
  NamedRerouteDeclaration(None=Multiply(A=LinearInterpolate(A=@n199, B=@n200, Alpha=@n206), B=NamedRerouteUsage())) {name=HerbShading, nodeColor=<rgba 0.4297 0 1 1>}
  NamedRerouteDeclaration(None=@n159) {name=DOTforShading, nodeColor=<rgba 0 1 0.3438 1>}
  NamedRerouteDeclaration(None=@n81) {name=TranslucentMat, nodeColor=<rgba 0.2188 0 1 1>}
  NamedRerouteDeclaration(None=@n143.RGB) {name=VirtualPlaneTextureRGB, nodeColor=<rgba 0 1 0.0391 1>}
  NamedRerouteDeclaration(None=VertexColor().R) {name=PerLeafRandom, nodeColor=<rgba 1 0 0.6308 1>}
  NamedRerouteDeclaration(None=@n198) {name=PerActorRandom, nodeColor=<rgba 0 1 0.7891 1>}
  n207 = NamedRerouteUsage()
  n208 = ScalarParameter() {parameterName=InteriorLines_MinOpacity}
  NamedRerouteDeclaration(None=LinearInterpolate(B=@n26, Alpha=Max(A=@n207, B=@n208)) {constA=1}) {name=Interior Lines Virtual PlaneBillboard, nodeColor=<rgba 1 0 0.3984 1>}
  NamedRerouteDeclaration(None=@n80.OveraAllBend) {name=Anim_WPO_OverallBend, nodeColor=<rgba 0.8047 0 1 1>}
  n209 = Subtract(A=PixelDepth().R, B=ScalarParameter() {defaultValue=400, group=Outlines, parameterName=InteriorLines_Fade_Offset})
  n210 = ScalarParameter() {defaultValue=1600, group=Outlines, parameterName=InteriorLines_Fade_Distance}
  NamedRerouteDeclaration(None=Saturate(None=Divide(A=@n209, B=@n210))) {name=NearFarMask, nodeColor=<rgba 0 1 0.2031 1>}
  n211 = DotProduct(A=@n103, B=@n101)
  n212 = MaterialFunctionCall(Multiplier=Constant() {r=0.5}) {materialFunction=&Pi}
  n213 = RotateAboutAxis(
   NormalizedRotationAxis=Normalize(VectorInput=@n104)
   RotationAngle=Multiply(A=Arccosine!(None=@n211), B=Divide(A=@n212.Result) {constB=10})
   PivotPoint=@n97
   Position=@n105.XYZ
  )
  NamedRerouteDeclaration(None=Add(A=NamedRerouteUsage(), B=NamedRerouteDeclaration(None=@n213) {name=BillboardTransformFallback, nodeColor=<rgba 1 0.1641 0 1>})) {name=WPO BillboardTransformFallback, nodeColor=<rgba 1 0.1641 0 1>}
  NamedRerouteUsage()
  NamedRerouteUsage()
  NamedRerouteUsage()
  NamedRerouteUsage()
  NamedRerouteUsage()
  NaniteReplace()
  Normalize(VectorInput=@n7)
  ObjectPositionWS()
  ObjectPositionWS()
  ObjectRadius()
  Power(Base=@n30, Exp=@n28)
  Round(None=@n39)
  Round(None=@n189)
  n214 = Multiply(A=@n18) {constB=-1}
  n215 = ScalarParameter() {defaultValue=104.7052, group=BillBoard WPO, parameterName=Billboard_EdgeMask}
  Saturate(None=Divide(A=Subtract(A=@n214, B=@n215)) {constB=0.0001})
  ScalarParameter() {defaultValue=1000, group=Color Variation WorldSpace, parameterName=Color Variation Distance WS, sortPriority=10}
  ScalarParameter() {defaultValue=1100, group=Color Variation WorldSpace, parameterName=Color Variation Distance WS2, sortPriority=10}
  SetMaterialAttributes!(MaterialAttributes=@n86, Emissive Color=NamedRerouteUsage()) {attributeSetTypes [#9]}
  ShadingModel()
  ShadowReplace(Default=@n213, Shadow=Constant3Vector())
  StaticSwitchParameter(True=Multiply(A=NamedRerouteUsage(), B=@n61), False=@n61) {defaultValue=true, parameterName=UseVirtual Plane Card}
  n216 = Multiply(A=@n17, B=Saturate(None=@n19))
  n217 = ScalarParameter() {defaultValue=-206.2373, group=BillBoard WPO, parameterName=BillboardSqueze_PushInner}
  StaticSwitchParameter(True=@n110, False=Add(A=@n110, B=Multiply(A=@n216, B=@n217))) {defaultValue=true, parameterName=UseVirtual Plane Card}
  StaticSwitchParameter(True=TwoSidedSign(), False=@n53) {parameterName=UseFlipBackFaceNormals}
  StaticSwitchParameter(True=@n83.Result, False=@n84) {group=Hedge, parameterName=HedgeShading, sortPriority=1}
  StaticSwitchParameter() {desc=Don't Use This is for the S37 Only, parameterName=IsUnlitMaterial}
  Subtract(
   A=WorldPosition().XYZ
   B=TransformPosition(None=Constant3Vector()) {transformSourceType=TRANSFORMPOSSOURCE_Instance, transformType=TRANSFORMPOSSOURCE_World}
  )
  n218 = TextureCoordinate() {coordinateIndex=2}
  n219 = AppendVector(
   A=ComponentMask(None=TextureCoordinate() {coordinateIndex=1}) {r=true}
   B=AppendVector(A=ComponentMask(None=@n218) {r=true}, B=ComponentMask(None=@n218) {g=true})
  )
  Subtract(A=Multiply(A=Frac(None=@n219)) {constB=12000}, B=Constant3Vector() {constant=<rgba 6000 6000 3000 0>}) {constB=6000}
  TextureSampleParameter2D() {
   group=Textures
   parameterName=Diffuse
   samplerSource=SSM_Wrap_WorldGroupSettings
   texture=&T_All_White
  }
  TransformPosition(None=NamedRerouteDeclaration(None=NamedRerouteUsage()) {name=LeafPivot Position LS, nodeColor=<rgba 0.0781 1 0 1>}) {transformType=TRANSFORMPOSSOURCE_World}
  TransformPosition(None=Constant3Vector()) {transformSourceType=TRANSFORMPOSSOURCE_Instance, transformType=TRANSFORMPOSSOURCE_World}
  Transform(None=@n11) {transformSourceType=TRANSFORMSOURCE_World, transformType=TRANSFORM_Camera}
  n220 = Custom() {
   code="//float3 Center = Parameters.SceneData.InstanceData.LocalBoundsCenter;\r\n//float3 Center = Parameters.SceneData.Primitive.MaxWPOExtent;\r\n\r\nfloat3 Center = Parameters.SceneData.InstanceData.LocalBoundsCenter;\r\nreturn Center;"
   showCode=true
  }
  VertexInterpolator(VS=@n220)
  WorldPosition()
 ]
}
